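# Niagara system for the Sheer Will glider cosmetic, drawing trails and a glow sprite with light, plus disabled idle smoke.
# Mesh trail emitters burst ten particles that follow skeletal mesh bones or sockets with mirrored offsets, faded by smoothed region masks and emitter age.
# VariantColorIndex picks Blot, Citadel or Froth colors and intensities, while FrontEnd modifiers retune values outside gameplay.
# Trail meshes scale to zero on large per-frame position jumps.
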NiagaraSystem /BRCosmetics/Gadgets/Assets/VinderTech_GliderChute/Glider_SheerWill/FX/NS_Glider_SheerWill
dependencies {
 usedDataInterfaces [
  &NiagaraDataInterfaceArrayFloat3
  &NiagaraDataInterfaceCurve
  &NiagaraDataInterfaceArrayFloat
  &NiagaraDataInterfaceSpriteRendererInfo
  &NiagaraDataInterfacePlatformSet
  &NiagaraDataInterfaceVector2DCurve
  &NiagaraDataInterfaceColorCurve
  &NiagaraDataInterfaceSkeletalMesh
 ]
 usedDynamicInputs [
  /Game/Effects/Niagara/DynamicInputs/Logic/FN_PickLinearColorDependOnPlatform
  &MakeCustomLinearColorFromBool
 ]
 usedModules [
  &self:SystemSpawnScript.SystemScriptSource.SystemScriptGraph.NiagaraNodeAssignment_0.SetVariables_#1
  &SystemState
  &self:SystemSpawnScript.SystemScriptSource.SystemScriptGraph.NiagaraNodeAssignment_1.SetVariables_#2
  &self:SystemSpawnScript.SystemScriptSource.SystemScriptGraph.NiagaraNodeAssignment_2.SetVariables_#3
  &self:SystemSpawnScript.SystemScriptSource.SystemScriptGraph.NiagaraNodeAssignment_3.SetVariables_#4
  &self:SystemSpawnScript.SystemScriptSource.SystemScriptGraph.NiagaraNodeAssignment_4.SetVariables_#5
  &self:SystemSpawnScript.SystemScriptSource.SystemScriptGraph.NiagaraNodeAssignment_5.SetVariables_#6
  &self:SystemSpawnScript.SystemScriptSource.SystemScriptGraph.NiagaraNodeAssignment_6.SetVariables_#7
  &self:SystemSpawnScript.SystemScriptSource.SystemScriptGraph.NiagaraNodeAssignment_7.SetVariables_#8
 ]
}
emitters {
 ForWardTrailLong {
  emitterUpdateScript [
   /SetVariables(Emitter.Des:NiagaraFloat=&self:ScratchDynamicInput_08(
     Blot:NiagaraFloat=0.5
     Citadel:NiagaraFloat=0.5
     Froth:NiagaraFloat=1
     VariantColorIndex:NiagaraInt32=&MakeIntFromFloat(Float=$User.VariantColorIndex)
    )) {isSetParameters=true}
   &EmitterState(Life Cycle Mode=::NewEnumerator1, Loop Behavior=::NewEnumerator1, Loop Duration Mode=::NewEnumerator1, UseLoopDelay=true)
   &SpawnBurst_Instantaneous(Spawn Count=10, Use Spawn Probability=false, UseLoopCountLimit=false)
  ]
  particleSpawnScript [
   &PartitionParticles(Number Of Partitions=10)
   &InitializeParticle(Color=<rgba 2 2 2 2>, Mesh Scale=<1.2 1.2 1>, Mesh Scale Mode=::NewEnumerator4, UsePositionOffset=false, Write Material Random=true)
   &InitialMeshOrientation(Mesh Orientation Mode=::NewEnumerator7, Rotation=<0 0 0.5>, Use Rotation Vector=true) {module=InitialMeshOrientation001}
   &InitialMeshOrientation(
    Mesh Orientation Mode=::NewEnumerator7
    Rotation=&Multiply_VectorByFloat(
     Vector=&SelectVectorFromArray(
      Direct Array Index=$Particles.PartitionParticles.Partition
      Vector Selection Array=DataInterface() {propertyValues {FloatData [<0 0 0.1>, <0 0 -0.1>, <0 0 0.1>, <0 0 -0.1>, <0 0 0.1>, <0 0 -0.1>, <0 0 0.1>, <0 0 -0.1>, <0 0 0.1>, <0 0 -0.1>]}}
     )
     Float=&FN_FrontEndFloatModifier(Gameplay=1.4, Front End=1)
    )
    Use Rotation Vector=true
   ) {module=InitialMeshOrientation002}
   &InitialMeshOrientation(
    Mesh Orientation Mode=::NewEnumerator7
    Rotation=&FN_FrontEndVector3Modifier(Gameplay=<-0.1 0 0>, Front End=<0.02 0 0>)
    Use Rotation Vector=true
   )
   &CameraOffset(Camera Offset Amount=-30)
   /SetVariables() {isSetParameters=true, module=SetVariables_#9}
  ]
  particleUpdateScript [
   /SetVariables(Particles.AlphaAll:NiagaraFloat=&Multiply_Float(
     A=&FloatFromCurve(
      FloatCurve=DataInterface() {propertyValues {Curve {defaultValue=1, keys [{tangentMode=RCTM_User, time=0, value=0}, {tangentMode=RCTM_User, time=1, value=1}]}}}
      CurveIndex=$Emitter.Age
      Scale Curve=&FN_FrontEndFloatModifier(Gameplay=1, Front End=0.9)
     )
     B=&FN_FrontEndFloatModifier(Gameplay=$System.InterpolateOverTime.InterpolateOverTime_Float.Moving Average, Front End=1)
    )) {isSetParameters=true, module=SetVariables_#10}
   &ParticleState(Kill Particles When Lifetime Has Elapsed=false)
   &SkeletalMeshLocation(
    Bone / Socket Index=$Particles.PartitionParticles.Partition
    Bone Sampling Mode=::NewEnumerator4
    MaskSpawnGroup=false
    Sampled Position Offset=&AddVectorToPosition(
     Position=&Multiply_VectorByFloat(
      Vector=&SelectVectorFromArray(
       Direct Array Index=$Particles.PartitionParticles.Partition
       Vector Selection Array=DataInterface() {propertyValues {FloatData [<-40 0 0>, <40 0 0>, <-40 0 0>, <40 0 0>, <-40 0 0>, <40 0 0>, <-40 0 0>, <40 0 0>, <-40 0 0>, <40 0 0>]}}
      )
      Float=&FN_FrontEndFloatModifier(Gameplay=1, Front End=0.7)
     )
     Vector=&FN_FrontEndVector3Modifier(Gameplay=<0 -30 -20>, Front End=<0 -30 -10>)
    )
    Skeletal Mesh=$System.Skeletal Mesh Reader
   )
   &RegionMask(
    Offset:Vector3f=$System.RegionMaskOffset
    Region Falloff:NiagaraFloat=$System.RegionMaskFade
    Region Radius:NiagaraFloat=$System.RegionMaskRadius
    UseOffset=true
    UseRotation=false
   )
   &RegionMask(
    Offset:Vector3f=$System.RegionMaskOffset
    Region Falloff=0
    Region Radius:NiagaraFloat=$System.RegionMaskRadius
    UseOffset=true
    UseRotation=false
   ) {module=RegionMask001}
   &InterpolateOverTime(Rate Of Change=10, Target Float Value:NiagaraFloat=$Particles.RegionMask.Float)
   &ScaleMeshSize(Scale Factor=&Multiply_VectorByFloat(
     Vector=&Multiply_Vector(
      A=&SelectVectorFromArray(
       Direct Array Index=$Particles.PartitionParticles.Partition
       Vector Selection Array=DataInterface() {propertyValues {FloatData [<-1 1 1>, <1 1 1>, <-1 1 1>, <1 1 1>, <-1 1 1>, <1 1 1>, <-1 1 1>, <1 1 1>, <-1 1 1>, <1 1 1>]}}
      )
      B=<1.5 2 1.5>
     )
     Float=&Multiply_Float(
      A=&RandomRangeFloat(Minimum=1, Maximum=1.1)
      B=&MakeCustomFloatFromBool(
       Bool=&SetBoolByFloatComparison(
        A=&VectorLength(Vector=&Subtract_Vector(A=Linked() {linkedVariable {name=Particles.Position, type=NiagaraPosition}}, B=$Particles.Lastposition))
        B=100
       )
       True Float=0
       False Float=1
      )
     )
    ))
   &DynamicMaterialParameters(
    Index 0 Param 1=&RandomRangeFloat(Minimum=1.5, Maximum=2.5)
    Index 0 Param 2=&RandomRangeFloat(Minimum=0, Maximum=1)
    Index 0 Param 4=&FloatFromCurve(
     FloatCurve=DataInterface() {propertyValues {Curve {defaultValue=1, keys [{tangentMode=RCTM_User, time=0, value=0}, {tangentMode=RCTM_User, time=1, value=1}]}}}
     CurveIndex=$Emitter.Age
    )
    Module.Index0Float0Write=true
    Module.Index0Float1Write=true
    Module.Index0Float2Write=true
    Module.Index0Float3Write=true
   )
   /SetVariables(Particles.Lastposition:NiagaraPosition=$Particles.Position) {isSetParameters=true, module=SetVariables_#11}
   &ScaleColor(
    Scale Alpha=&Multiply_Float(
     A=&SelectFloatFromArray(
      Direct Array Index=$Particles.PartitionParticles.Partition
      Float Selection Array=DataInterface() {propertyValues {FloatData [1,1,1,1,1,1,1,1,1,1]}}
     )
     B=$Particles.InterpolateOverTime.InterpolateOverTime_Float.Moving Average
    )
    ScaleA=true
    ScaleRGB=true
   )
   &ScaleColor(
    Scale Alpha=&Multiply_Float(A=$Particles.RegionMask001.Float, B=$Particles.AlphaAll)
    ScaleA=true
    ScaleRGB=true
   ) {module=ScaleColor001}
   &FN_Mesh_Attributes(
    Color=&MakeLinearColorFromVectorAndFloat(
     "Vector (RGB)"=&Multiply_Vector(
      A=&MakeVectorFromLinearColorRGB(Color=&self:ScratchDynamicInput_01(
        BlotColor:LinearColor=<rgba 1 0.0146 0 0.4>
        CitadelColor:LinearColor=<rgba 0 0.6667 2 0.4>
        FrothColor:LinearColor=<rgba 0 0 0 1>
        VariantColorIndex:NiagaraInt32=&MakeIntFromFloat(Float=$User.VariantColorIndex)
       ))
      B=<1 1 1>
     )
     "Float (Alpha)"=&Multiply_Float(
      A=&Multiply_Float(
       A=&Multiply_Float(
        A=&Multiply_Float(
         A=0.6
         B=&self:ScratchDynamicInput_10(
          Blot:NiagaraFloat=1
          Citadel:NiagaraFloat=$System.FECitadelGlow
          Froth:NiagaraFloat=&FN_FrontEndFloatModifier(Gameplay=3, Front End=20)
          VariantColorIndex:NiagaraInt32=&MakeIntFromFloat(Float=$User.VariantColorIndex)
         )
        )
        B=&Multiply_Float(A=&FN_FrontEndFloatModifier(Gameplay=1, Front End=0.2), B=$Particles.AlphaAll)
       )
       B=$Particles.InterpolateOverTime.InterpolateOverTime_Float.Moving Average
      )
      B=&Multiply_Float(A=$Particles.RegionMask001.Float, B=$Particles.AlphaAll)
     )
    )
    Create Color Attribute=true
    Module Output=::NewEnumerator1
   ) {module=FN_Mesh_Attributes001}
   &FN_Mesh_Attributes(
    Color=&MakeLinearColorFromVectorAndFloat(
     "Vector (RGB)"=&Multiply_Vector(
      A=&MakeVectorFromLinearColorRGB(Color=&self:ScratchDynamicInput_01(
        BlotColor:LinearColor=<rgba 1 0.0146 0 0.4>
        CitadelColor:LinearColor=<rgba 0 0.6667 2 0.4>
        FrothColor:LinearColor=<rgba 0 0 0 1>
        VariantColorIndex:NiagaraInt32=&MakeIntFromFloat(Float=$User.VariantColorIndex)
       ))
      B=<1 1 1>
     )
     "Float (Alpha)"=&Multiply_Float(
      A=&Multiply_Float(
       A=&Multiply_Float(
        A=&Multiply_Float(
         A=0.1
         B=&self:ScratchDynamicInput_11(
          Blot:NiagaraFloat=1
          Citadel:NiagaraFloat=$System.FECitadelGlow
          Froth:NiagaraFloat=&FN_FrontEndFloatModifier(Gameplay=3, Front End=20)
          VariantColorIndex:NiagaraInt32=&MakeIntFromFloat(Float=$User.VariantColorIndex)
         )
        )
        B=&Multiply_Float(A=&FN_FrontEndFloatModifier(Gameplay=1, Front End=0.2), B=$Particles.AlphaAll)
       )
       B=$Particles.InterpolateOverTime.InterpolateOverTime_Float.Moving Average
      )
      B=&Multiply_Float(A=$Particles.RegionMask001.Float, B=$Particles.AlphaAll)
     )
    )
    Create Color Attribute=true
    Module Output=::NewEnumerator1
   ) {module=FN_Mesh_Attributes002}
  ]
  rendererClasses [&NiagaraMeshRendererProperties]
  renderers [
   {rendererClass=&NiagaraMeshRendererProperties, rendererIndex=0}
   {rendererClass=&NiagaraMeshRendererProperties, rendererIndex=1}
   {rendererClass=&NiagaraMeshRendererProperties, rendererIndex=2}
  ]
 }
 ForWardTrailLong003 {
  emitterUpdateScript [
   /SetVariables(Emitter.Des:NiagaraFloat=&self:ScratchDynamicInput_08(
     Blot:NiagaraFloat=0.5
     Citadel:NiagaraFloat=0.5
     Froth:NiagaraFloat=1
     VariantColorIndex:NiagaraInt32=&MakeIntFromFloat(Float=$User.VariantColorIndex)
    )) {isSetParameters=true}
   &EmitterState(Life Cycle Mode=::NewEnumerator1, Loop Behavior=::NewEnumerator1, Loop Duration Mode=::NewEnumerator1, UseLoopDelay=true)
   &SpawnBurst_Instantaneous(
    Spawn Count=10
    Spawn Probability=&FN_FrontEndFloatModifier(Gameplay=1, Front End=0)
    Use Spawn Probability=true
    UseLoopCountLimit=false
   )
  ]
  particleSpawnScript [
   &PartitionParticles(Number Of Partitions=10)
   &InitializeParticle(Color=<rgba 2 2 2 2>, Mesh Scale=<1.2 1.2 1>, Mesh Scale Mode=::NewEnumerator4, UsePositionOffset=false, Write Material Random=true)
   &InitialMeshOrientation(Mesh Orientation Mode=::NewEnumerator7, Rotation=<0 0 0.5>, Use Rotation Vector=true) {module=InitialMeshOrientation001}
   &InitialMeshOrientation(
    Mesh Orientation Mode=::NewEnumerator7
    Rotation=&Multiply_VectorByFloat(
     Vector=&SelectVectorFromArray(
      Direct Array Index=$Particles.PartitionParticles.Partition
      Vector Selection Array=DataInterface() {propertyValues {FloatData [<0 0 0.1>, <0 0 -0.1>, <0 0 0.1>, <0 0 -0.1>, <0 0 0.1>, <0 0 -0.1>, <0 0 0.1>, <0 0 -0.1>, <0 0 0.1>, <0 0 -0.1>]}}
     )
     Float=&FN_FrontEndFloatModifier(Gameplay=1.5, Front End=1)
    )
    Use Rotation Vector=true
   ) {module=InitialMeshOrientation002}
   &InitialMeshOrientation(
    Mesh Orientation Mode=::NewEnumerator7
    Rotation=&FN_FrontEndVector3Modifier(Gameplay=<0 0 0>, Front End=<0.05 0 0>)
    Use Rotation Vector=true
   )
   &CameraOffset(Camera Offset Amount=-30)
   /SetVariables() {isSetParameters=true, module=SetVariables_#9}
  ]
  particleUpdateScript [
   /SetVariables(Particles.BrightnessAll:NiagaraFloat=&self:ScratchDynamicInput_09(
     Blot:NiagaraFloat=1
     Citadel:NiagaraFloat=1
     Froth:NiagaraFloat=1
     VariantColorIndex:NiagaraInt32=&MakeIntFromFloat(Float=$User.VariantColorIndex)
    )) {isSetParameters=true}
   /SetVariables(Particles.AlphaAll:NiagaraFloat=&Multiply_Float(
     A=&FloatFromCurve(
      FloatCurve=DataInterface() {propertyValues {Curve {defaultValue=1, keys [{tangentMode=RCTM_User, time=0, value=0}, {tangentMode=RCTM_User, time=1, value=1}]}}}
      CurveIndex=$Emitter.Age
      Scale Curve=0.7
     )
     B=&FN_FrontEndFloatModifier(Gameplay=$System.InterpolateOverTime.InterpolateOverTime_Float.Moving Average, Front End=1)
    )) {isSetParameters=true, module=SetVariables_#10}
   &ParticleState(Kill Particles When Lifetime Has Elapsed=false)
   &SkeletalMeshLocation(
    Bone / Socket Index=$Particles.PartitionParticles.Partition
    Bone Sampling Mode=::NewEnumerator4
    MaskSpawnGroup=false
    Sampled Position Offset=&AddVectorToPosition(
     Position=&Multiply_VectorByFloat(
      Vector=&SelectVectorFromArray(
       Direct Array Index=$Particles.PartitionParticles.Partition
       Vector Selection Array=DataInterface() {propertyValues {FloatData [<-40 0 0>, <40 0 0>, <-40 0 0>, <40 0 0>, <-40 0 0>, <40 0 0>, <-40 0 0>, <40 0 0>, <-40 0 0>, <40 0 0>]}}
      )
      Float=&FN_FrontEndFloatModifier(Gameplay=1.3, Front End=0.7)
     )
     Vector=&FN_FrontEndVector3Modifier(Gameplay=<0 -110 -45>, Front End=<0 -110 -10>)
    )
    Skeletal Mesh=$System.Skeletal Mesh Reader
   )
   &RegionMask(
    Offset:Vector3f=$System.RegionMaskOffset
    Region Falloff:NiagaraFloat=$System.RegionMaskFade
    Region Radius:NiagaraFloat=$System.RegionMaskRadius
    UseOffset=true
    UseRotation=false
   )
   &RegionMask(
    Offset:Vector3f=$System.RegionMaskOffset
    Region Falloff=0
    Region Radius:NiagaraFloat=$System.RegionMaskRadius
    UseOffset=true
    UseRotation=false
   ) {module=RegionMask001}
   &InterpolateOverTime(Rate Of Change=10, Target Float Value:NiagaraFloat=$Particles.RegionMask.Float)
   &ScaleMeshSize(Scale Factor=&Multiply_VectorByFloat(
     Vector=&Multiply_Vector(
      A=&SelectVectorFromArray(
       Direct Array Index=$Particles.PartitionParticles.Partition
       Vector Selection Array=DataInterface() {propertyValues {FloatData [<-1 1 1>, <1 1 1>, <-1 1 1>, <1 1 1>, <-1 1 1>, <1 1 1>, <-1 1 1>, <1 1 1>, <-1 1 1>, <1 1 1>]}}
      )
      B=<1.5 2 1.5>
     )
     Float=&Multiply_Float(
      A=&RandomRangeFloat(Minimum=0.8, Maximum=1)
      B=&MakeCustomFloatFromBool(
       Bool=&SetBoolByFloatComparison(
        A=&VectorLength(Vector=&Subtract_Vector(A=Linked() {linkedVariable {name=Particles.Position, type=NiagaraPosition}}, B=$Particles.Lastposition))
        B=100
       )
       True Float=0
       False Float=1
      )
     )
    ))
   &DynamicMaterialParameters(
    Index 0 Param 1=&RandomRangeFloat(Minimum=1.3, Maximum=1.3)
    Index 0 Param 2=&RandomRangeFloat(Minimum=0, Maximum=1)
    Index 0 Param 4=&FloatFromCurve(
     FloatCurve=DataInterface() {propertyValues {Curve {defaultValue=1, keys [{tangentMode=RCTM_User, time=0, value=0}, {tangentMode=RCTM_User, time=1, value=1}]}}}
     CurveIndex=$Emitter.Age
    )
    Module.Index0Float0Write=true
    Module.Index0Float1Write=true
    Module.Index0Float2Write=true
    Module.Index0Float3Write=true
   )
   /SetVariables(Particles.Lastposition:NiagaraPosition=$Particles.Position) {isSetParameters=true, module=SetVariables_#11}
   &ScaleColor(
    Scale Alpha=&Multiply_Float(
     A=&SelectFloatFromArray(
      Direct Array Index=$Particles.PartitionParticles.Partition
      Float Selection Array=DataInterface() {propertyValues {FloatData [1,1,1,1,1,1,1,1,1,1]}}
     )
     B=$Particles.InterpolateOverTime.InterpolateOverTime_Float.Moving Average
    )
    ScaleA=true
    ScaleRGB=true
   )
   &ScaleColor(
    Scale Alpha=&Multiply_Float(A=$Particles.RegionMask001.Float, B=$Particles.AlphaAll)
    Scale RGB=&Multiply_VectorByFloat(Vector=<1 1 1>, Float=$Particles.BrightnessAll)
    ScaleA=true
    ScaleRGB=true
   ) {module=ScaleColor001}
   &FN_Mesh_Attributes(
    Color=&MakeLinearColorFromVectorAndFloat(
     "Vector (RGB)"=&Multiply_Vector(
      A=&MakeVectorFromLinearColorRGB(Color=&self:ScratchDynamicInput_01(
        BlotColor:LinearColor=<rgba 1 0.0146 0 0.4>
        CitadelColor:LinearColor=<rgba 0 1 3 0.4>
        FrothColor:LinearColor=<rgba 0 0 0 1>
        VariantColorIndex:NiagaraInt32=&MakeIntFromFloat(Float=$User.VariantColorIndex)
       ))
      B=&Multiply_VectorByFloat(Vector=<1 1 1>, Float=$Particles.BrightnessAll)
     )
     "Float (Alpha)"=&Multiply_Float(
      A=&Multiply_Float(
       A=&Multiply_Float(
        A=&Multiply_Float(
         A=0.6
         B=&self:ScratchDynamicInput_10(
          Blot:NiagaraFloat=1
          Citadel:NiagaraFloat=$System.FECitadelGlow
          Froth:NiagaraFloat=3
          VariantColorIndex:NiagaraInt32=&MakeIntFromFloat(Float=$User.VariantColorIndex)
         )
        )
        B=&Multiply_Float(A=&FN_FrontEndFloatModifier(Gameplay=1, Front End=0.2), B=$Particles.AlphaAll)
       )
       B=$Particles.InterpolateOverTime.InterpolateOverTime_Float.Moving Average
      )
      B=&Multiply_Float(A=$Particles.RegionMask001.Float, B=$Particles.AlphaAll)
     )
    )
    Create Color Attribute=true
    Module Output=::NewEnumerator1
   ) {module=FN_Mesh_Attributes001}
   &FN_Mesh_Attributes(
    Color=&MakeLinearColorFromVectorAndFloat(
     "Vector (RGB)"=&Multiply_Vector(
      A=&MakeVectorFromLinearColorRGB(Color=&self:ScratchDynamicInput_01(
        BlotColor:LinearColor=<rgba 1 0.0146 0 0.4>
        CitadelColor:LinearColor=<rgba 0 1 3 0.4>
        FrothColor:LinearColor=<rgba 0 0 0 1>
        VariantColorIndex:NiagaraInt32=&MakeIntFromFloat(Float=$User.VariantColorIndex)
       ))
      B=&Multiply_VectorByFloat(Vector=<1 1 1>, Float=$Particles.BrightnessAll)
     )
     "Float (Alpha)"=&Multiply_Float(
      A=&Multiply_Float(
       A=&Multiply_Float(
        A=&Multiply_Float(
         A=0.1
         B=&self:ScratchDynamicInput_11(
          Blot:NiagaraFloat=1
          Citadel:NiagaraFloat=$System.FECitadelGlow
          Froth:NiagaraFloat=3
          VariantColorIndex:NiagaraInt32=&MakeIntFromFloat(Float=$User.VariantColorIndex)
         )
        )
        B=&Multiply_Float(A=&FN_FrontEndFloatModifier(Gameplay=1, Front End=0.2), B=$Particles.AlphaAll)
       )
       B=$Particles.InterpolateOverTime.InterpolateOverTime_Float.Moving Average
      )
      B=&Multiply_Float(A=$Particles.RegionMask001.Float, B=$Particles.AlphaAll)
     )
    )
    Create Color Attribute=true
    Module Output=::NewEnumerator1
   ) {module=FN_Mesh_Attributes002}
  ]
  rendererClasses [&NiagaraMeshRendererProperties]
  renderers [
   {rendererClass=&NiagaraMeshRendererProperties, rendererIndex=0}
   {rendererClass=&NiagaraMeshRendererProperties, rendererIndex=1}
   {rendererClass=&NiagaraMeshRendererProperties, rendererIndex=2}
  ]
 }
 ForWardTrailShort {
  emitterUpdateScript [
   /SetVariables(Emitter.Des:NiagaraFloat=&self:ScratchDynamicInput_08(
     Blot:NiagaraFloat=0.5
     Citadel:NiagaraFloat=0.5
     Froth:NiagaraFloat=1
     VariantColorIndex:NiagaraInt32=&MakeIntFromFloat(Float=$User.VariantColorIndex)
    )) {isSetParameters=true}
   &EmitterState(Life Cycle Mode=::NewEnumerator1, Loop Behavior=::NewEnumerator1, Loop Duration Mode=::NewEnumerator1, UseLoopDelay=true)
   &SpawnBurst_Instantaneous(Spawn Count=10, Use Spawn Probability=false, UseLoopCountLimit=false)
  ]
  particleSpawnScript [
   &PartitionParticles(Number Of Partitions=10)
   &InitializeParticle(Color=<rgba 2 2 2 2>, Mesh Scale=<1.2 1.2 1>, Mesh Scale Mode=::NewEnumerator4, UsePositionOffset=false, Write Material Random=true)
   &InitialMeshOrientation(Mesh Orientation Mode=::NewEnumerator7, Rotation=<0 0 0.5>, Use Rotation Vector=true) {module=InitialMeshOrientation001}
   &InitialMeshOrientation(
    Mesh Orientation Mode=::NewEnumerator7
    Rotation=&Multiply_VectorByFloat(
     Vector=&SelectVectorFromArray(
      Direct Array Index=$Particles.PartitionParticles.Partition
      Vector Selection Array=DataInterface() {propertyValues {FloatData [<0 0 0.1>, <0 0 -0.1>, <0 0 0.1>, <0 0 -0.1>, <0 0 0.1>, <0 0 -0.1>, <0 0 0.1>, <0 0 -0.1>, <0 0 0.1>, <0 0 -0.1>]}}
     )
     Float=&FN_FrontEndFloatModifier(Gameplay=1.2, Front End=0.9)
    )
    Use Rotation Vector=true
   ) {module=InitialMeshOrientation002}
   &InitialMeshOrientation(
    Mesh Orientation Mode=::NewEnumerator7
    Rotation=&FN_FrontEndVector3Modifier(Gameplay=<-0.1 0 0>, Front End=<0.02 0 0>)
    Use Rotation Vector=true
   )
   &CameraOffset(Camera Offset Amount=-30)
   /SetVariables() {isSetParameters=true, module=SetVariables_#9}
  ]
  particleUpdateScript [
   /SetVariables(Particles.AlphaAll:NiagaraFloat=&Multiply_Float(
     A=&FloatFromCurve(
      FloatCurve=DataInterface() {propertyValues {Curve {defaultValue=1, keys [{tangentMode=RCTM_User, time=0, value=0}, {tangentMode=RCTM_User, time=1, value=1}]}}}
      CurveIndex=$Emitter.Age
      Scale Curve=0.8
     )
     B=&FN_FrontEndFloatModifier(Gameplay=$System.InterpolateOverTime.InterpolateOverTime_Float.Moving Average, Front End=1)
    )) {isSetParameters=true, module=SetVariables_#10}
   &DynamicMaterialParameters(
    Index 0 Param 1=&RandomRangeFloat(Minimum=1, Maximum=1.7)
    Index 0 Param 2=&RandomRangeFloat(Minimum=0, Maximum=1)
    Index 0 Param 4=&FloatFromCurve(
     FloatCurve=DataInterface() {propertyValues {Curve {defaultValue=1, keys [{tangentMode=RCTM_User, time=0, value=0}, {tangentMode=RCTM_User, time=1, value=1}]}}}
     CurveIndex=$Emitter.Age
    )
    Module.Index0Float0Write=true
    Module.Index0Float1Write=true
    Module.Index0Float2Write=true
    Module.Index0Float3Write=true
   )
   &ParticleState(Kill Particles When Lifetime Has Elapsed=false)
   &SkeletalMeshLocation(
    Bone / Socket Index=$Particles.PartitionParticles.Partition
    Bone Sampling Mode=::NewEnumerator4
    MaskSpawnGroup=false
    Sampled Position Offset=&AddVectorToPosition(
     Position=&Multiply_VectorByFloat(
      Vector=&SelectVectorFromArray(
       Direct Array Index=$Particles.PartitionParticles.Partition
       Vector Selection Array=DataInterface() {propertyValues {FloatData [<-40 0 0>, <40 0 0>, <-40 0 0>, <40 0 0>, <-40 0 0>, <40 0 0>, <-40 0 0>, <40 0 0>, <-40 0 0>, <40 0 0>]}}
      )
      Float=0.5
     )
     Vector=&FN_FrontEndVector3Modifier(Gameplay=<0 -30 -40>, Front End=<0 -30 -40>)
    )
    Skeletal Mesh=$System.Skeletal Mesh Reader
   )
   &RegionMask(
    Offset:Vector3f=$System.RegionMaskOffset
    Region Falloff:NiagaraFloat=$System.RegionMaskFade
    Region Radius:NiagaraFloat=$System.RegionMaskRadius
    UseOffset=true
    UseRotation=false
   )
   &RegionMask(
    Offset:Vector3f=$System.RegionMaskOffset
    Region Falloff=0
    Region Radius:NiagaraFloat=$System.RegionMaskRadius
    UseOffset=true
    UseRotation=false
   ) {module=RegionMask001}
   &InterpolateOverTime(Rate Of Change=10, Target Float Value:NiagaraFloat=$Particles.RegionMask.Float)
   &ScaleMeshSize(Scale Factor=&Multiply_VectorByFloat(
     Vector=&Multiply_Vector(
      A=&SelectVectorFromArray(
       Direct Array Index=$Particles.PartitionParticles.Partition
       Vector Selection Array=DataInterface() {propertyValues {FloatData [<-1 1 1>, <1 1 1>, <-1 1 1>, <1 1 1>, <-1 1 1>, <1 1 1>, <-1 1 1>, <1 1 1>, <-1 1 1>, <1 1 1>]}}
      )
      B=&Multiply_VectorByFloat(Vector=<1.5 1.7 1.5>, Float=0.8)
     )
     Float=&Multiply_Float(
      A=&RandomRangeFloat(Minimum=1.1, Maximum=1.1)
      B=&MakeCustomFloatFromBool(
       Bool=&SetBoolByFloatComparison(
        A=&VectorLength(Vector=&Subtract_Vector(A=Linked() {linkedVariable {name=Particles.Position, type=NiagaraPosition}}, B=$Particles.Lastposition))
        B=100
       )
       True Float=0
       False Float=1
      )
     )
    ))
   /SetVariables(Particles.Lastposition:NiagaraPosition=$Particles.Position) {isSetParameters=true, module=SetVariables_#11}
   &ScaleColor(
    Scale Alpha=&Multiply_Float(
     A=&SelectFloatFromArray(
      Direct Array Index=$Particles.PartitionParticles.Partition
      Float Selection Array=DataInterface() {propertyValues {FloatData [1,1,1,1,1,1,1,1,1,1]}}
     )
     B=$Particles.InterpolateOverTime.InterpolateOverTime_Float.Moving Average
    )
    ScaleA=true
    ScaleRGB=true
   )
   &ScaleColor(
    Scale Alpha=&Multiply_Float(A=$Particles.RegionMask001.Float, B=$Particles.AlphaAll)
    ScaleA=true
    ScaleRGB=true
   ) {module=ScaleColor001}
   &FN_Mesh_Attributes(
    Color=&MakeLinearColorFromVectorAndFloat(
     "Vector (RGB)"=&Multiply_Vector(
      A=&MakeVectorFromLinearColorRGB(Color=&self:ScratchDynamicInput_01(
        BlotColor:LinearColor=<rgba 1 0.0146 0 0.4>
        CitadelColor:LinearColor=<rgba 0 1 3 0.4>
        FrothColor:LinearColor=<rgba 0 0 0 1>
        VariantColorIndex:NiagaraInt32=&MakeIntFromFloat(Float=$User.VariantColorIndex)
       ))
      B=<1 1 1>
     )
     "Float (Alpha)"=&Multiply_Float(
      A=&Multiply_Float(
       A=&Multiply_Float(
        A=&Multiply_Float(
         A=0.6
         B=&self:ScratchDynamicInput_10(
          Blot:NiagaraFloat=1
          Citadel:NiagaraFloat=$System.FECitadelGlow
          Froth:NiagaraFloat=&FN_FrontEndFloatModifier(Gameplay=3, Front End=20)
          VariantColorIndex:NiagaraInt32=&MakeIntFromFloat(Float=$User.VariantColorIndex)
         )
        )
        B=&Multiply_Float(A=&FN_FrontEndFloatModifier(Gameplay=1, Front End=0.2), B=$Particles.AlphaAll)
       )
       B=$Particles.InterpolateOverTime.InterpolateOverTime_Float.Moving Average
      )
      B=&Multiply_Float(A=$Particles.RegionMask001.Float, B=$Particles.AlphaAll)
     )
    )
    Create Color Attribute=true
    Module Output=::NewEnumerator1
   ) {module=FN_Mesh_Attributes001}
  ]
  rendererClasses [&NiagaraMeshRendererProperties]
  renderers [
   {rendererClass=&NiagaraMeshRendererProperties, rendererIndex=0}
   {rendererClass=&NiagaraMeshRendererProperties, rendererIndex=1}
  ]
 }
 Glow {
  emitterUpdateScript [
   &EmitterState(Life Cycle Mode=::NewEnumerator1, Loop Behavior=::NewEnumerator1, Loop Duration Mode=::NewEnumerator1, UseLoopDelay=false)
   &SpawnBurst_Instantaneous(Use Spawn Probability=false, UseLoopCountLimit=false)
  ]
  particleSpawnScript [
   &InitializeParticle(
    Color Mode=::NewEnumerator0
    Position Offset=&FN_FrontEndVector3Modifier(Gameplay=<0 50 -20>, Front End=<0 25 0>)
    Sprite Size Mode=::NewEnumerator3
    Uniform Sprite Size=&FN_FrontEndFloatModifier(Gameplay=200, Front End=150)
    UsePositionOffset=true
    Write Material Random=true
   )
   &DynamicMaterialParameters(
    Index 0 Param 1=1.3
    Index 0 Param 2=3
    Module.Index0Float0Write=true
    Module.Index0Float1Write=true
    Module.Index0Float2Write=true
    Module.Index0Float3Write=true
   )
  ]
  particleUpdateScript [
   &Color(Color=&self:ScratchDynamicInput_04(
     BlotColor:LinearColor=<rgba 0.2 0 0 0.9>
     CitadelColor:LinearColor=<rgba 0 0.1333 0.4 1>
     FrothColor:LinearColor=<rgba 0 0 0 0.8>
     VariantColorIndex:NiagaraInt32=&MakeIntFromFloat(Float=$User.VariantColorIndex)
    ))
   &ParticleState(Kill Particles When Lifetime Has Elapsed=false)
   &ScaleColor(
    Color Value To Scale=$Particles.Color
    Scale Alpha=&FloatFromCurve(
     FloatCurve=DataInterface() {propertyValues {Curve {defaultValue=1, keys [{tangentMode=RCTM_User, time=0, value=0}, {tangentMode=RCTM_User, time=1, value=1}]}}}
     CurveIndex=$Emitter.Age
     Scale Curve=&self:ScratchDynamicInput_13(
      Blot:NiagaraFloat=1
      Citadel:NiagaraFloat=1
      Froth:NiagaraFloat=&FN_FrontEndFloatModifier(Gameplay=1, Front End=1.3)
      VariantColorIndex:NiagaraInt32=&MakeIntFromFloat(Float=$User.VariantColorIndex)
     )
    )
    ScaleA=true
    ScaleRGB=true
   )
   &Light_Attributes(
    Color=1
    Color Mode=0
    Falloff Exponent=2
    Light Color:LinearColor=&Multiply_LinearColorByFloat(
     Linear Color=&self:ScratchDynamicInput_04(
      BlotColor:LinearColor=<rgba 20 0 0 1>
      CitadelColor:LinearColor=<rgba 0 13.3333 40 1>
      FrothColor:LinearColor=<rgba 0 0 0 1>
      VariantColorIndex:NiagaraInt32=&MakeIntFromFloat(Float=$User.VariantColorIndex)
     )
     Float=&FloatFromCurve(
      FloatCurve=DataInterface() {propertyValues {Curve {defaultValue=1, keys [{tangentMode=RCTM_User, time=0, value=0}, {tangentMode=RCTM_User, time=1, value=1}]}}}
      CurveIndex=$Emitter.Age
     )
    )
    Light Exponent=3
    Light Position Offset:Vector3f=<0 20 50>
    Light Radius=150
    Position=1
    Radius=2
    UsePositionOffset=true
   )
  ]
  rendererClasses [&NiagaraSpriteRendererProperties, &NiagaraLightRendererProperties]
  renderers [
   {rendererClass=&NiagaraSpriteRendererProperties}
   {rendererClass=&NiagaraLightRendererProperties, rendererIndex=1}
  ]
 }
 IdleSmoke {
  bEnabled=false
  emitterUpdateScript [
   &EmitterState(Life Cycle Mode=::NewEnumerator1, Loop Duration=1, UseLoopDelay=false)
   &SpawnRate(SpawnRate=8, Use Spawn Probability=true)
   &SpawnBurst_Instantaneous(Spawn Count=4, Use Spawn Probability=true, UseLoopCountLimit=false)
  ]
  particleSpawnScript [
   &PartitionParticles(Number Of Partitions=10)
   &InitializeParticle(
    Color=&LerpLinearColors(
     StartColor=<rgba 0.7 0.0255 0.0817 3>
     EndColor=<rgba 0.3 0.0109 0.035 3>
     LerpFactor=&UniformAOrBFloat(A=-1, Evaluation Type=::NewEnumerator0, Recalculate Random Each Loop=true, Override Seed=false)
    )
    Lifetime Max=2.5
    Lifetime Min=1.5
    Lifetime Mode=::NewEnumerator1
    Mesh Scale=<1.2 1.2 1>
    Mesh Scale Mode=::NewEnumerator4
    Sprite Rotation Mode=::NewEnumerator2
    Sprite Size Mode=::NewEnumerator1
    Uniform Sprite Size Max=300
    Uniform Sprite Size Min=200
    UsePositionOffset=false
    Write Material Random=true
   )
   &ShapeLocation(
    Box Size=<50 50 70>
    MaskSpawnGroup=false
    Override Seed=false
    Shape Origin=&VectorFromFloat(Value=0)
    Shape Primitive=::NewEnumerator2
    Surface Only=false
    UseOwnerScale=true
   )
   &CameraOffset(Camera Offset Amount=-30)
   /SetVariables() {isSetParameters=true, module=SetVariables_#9}
  ]
  particleUpdateScript [
   &DynamicMaterialParameters(
    Index 0 Param 1=&RandomRangeFloat(Minimum=0, Maximum=0)
    Index 0 Param 2=&RandomRangeFloat(Minimum=0, Maximum=1)
    Module.Index0Float0Write=true
    Module.Index0Float1Write=true
    Module.Index0Float2Write=true
    Module.Index0Float3Write=true
   ) {module=DynamicMaterialParameters001}
   &ParticleState()
   &SpriteRotationRate(Rotation Rate=&RandomRangeFloat(Minimum=-50, Maximum=50))
   &SkeletalMeshLocation(
    Bone / Socket Index=$Particles.PartitionParticles.Partition
    Bone Sampling Mode=::NewEnumerator4
    MaskSpawnGroup=false
    Sampled Position Offset=&AddVectorToPosition(
     Position=&AddVectorToPosition(
      Position=&SelectVectorFromArray(
       Direct Array Index=$Particles.PartitionParticles.Partition
       Vector Selection Array=DataInterface() {propertyValues {FloatData [<-30 0 0>, <20 0 0>, <-10 0 0>, <15 0 0>, <-10 0 0>, <15 0 0>, <-30 0 0>, <20 0 0>, <-30 0 0>, <20 0 0>]}}
      )
      Vector=Linked() {linkedVariable {name=Particles.ShapeLocation.ShapePosition, type=NiagaraPosition}}
     )
     Vector=<0 -30 -10>
    )
    Skeletal Mesh=$System.Skeletal Mesh Reader
   )
   &RegionMask(Offset:Vector3f=$System.RegionMaskOffset, Region Falloff=1, Region Radius=250, UseOffset=true, UseRotation=false) {module=RegionMask001}
   &ScaleSpriteSize(
    Scale Sprite Size Mode=::NewEnumerator0
    Uniform Scale Factor=&Multiply_Float(
     A=&FloatFromCurve(FloatCurve=DataInterface() {propertyValues {Curve {defaultValue=1, keys [{tangentMode=RCTM_User, time=0, value=0.5}, {tangentMode=RCTM_User, time=1, value=1}]}}})
     B=&Multiply_Float(
      A=0.8
      B=&MakeCustomFloatFromBool(
       Bool=&SetBoolByFloatComparison(
        A=&VectorLength(Vector=&Subtract_Vector(A=Linked() {linkedVariable {name=Particles.Position, type=NiagaraPosition}}, B=$Particles.Lastposition))
        B=100
       )
       True Float=0
       False Float=1
      )
     )
    )
   )
   &ScaleColor(
    Scale Alpha=&Multiply_Float(
     A=&FloatFromCurve(
      FloatCurve=DataInterface() {propertyValues {Curve {defaultValue=1, keys [{tangentMode=RCTM_User, time=0, value=0}, {tangentMode=RCTM_User, time=1, value=1}]}}}
      CurveIndex=$Emitter.Age
     )
     B=&FloatFromCurve(
      FloatCurve=DataInterface() {propertyValues {Curve {
         defaultValue=1
         keys [
          {interpMode=RCIM_Cubic, tangentMode=RCTM_User, time=0, value=0}
          {interpMode=RCIM_Cubic, tangentMode=RCTM_User, time=0.5, value=1}
          {tangentMode=RCTM_User, time=1, value=0}
         ]
        }}}
      Scale Curve=0.8
     )
    )
    Scale RGB=&Multiply_VectorByFloat(Vector=<1 1 1>, Float=1.3)
    ScaleA=true
    ScaleRGB=true
   )
   /SetVariables(Particles.Lastposition:NiagaraPosition=$Particles.Position) {isSetParameters=true, module=SetVariables_#11}
   &ScaleColor(Scale Alpha=$Particles.RegionMask001.Float, ScaleA=true, ScaleRGB=true) {module=ScaleColor001}
  ]
  rendererClasses [&NiagaraSpriteRendererProperties]
  renderers [{rendererClass=&NiagaraSpriteRendererProperties}]
 }
 IdleTrailInside {
  emitterUpdateScript [
   /SetVariables(Emitter.Des:NiagaraFloat=&self:ScratchDynamicInput_08(
     Blot:NiagaraFloat=0.5
     Citadel:NiagaraFloat=0.5
     Froth:NiagaraFloat=1
     VariantColorIndex:NiagaraInt32=&MakeIntFromFloat(Float=$User.VariantColorIndex)
    )) {isSetParameters=true}
   &EmitterState(Life Cycle Mode=::NewEnumerator1, Loop Behavior=::NewEnumerator1, Loop Duration Mode=::NewEnumerator1, UseLoopDelay=true)
   &SpawnBurst_Instantaneous(Spawn Count=10, Use Spawn Probability=false, UseLoopCountLimit=false)
  ]
  particleSpawnScript [
   &PartitionParticles(Number Of Partitions=10)
   &InitializeParticle(Color=<rgba 2 2 2 2>, Mesh Scale=<1.2 1.2 1>, Mesh Scale Mode=::NewEnumerator4, UsePositionOffset=false, Write Material Random=true)
   &InitialMeshOrientation(Mesh Orientation Mode=::NewEnumerator7, Rotation=<0.05 0 0>, Use Rotation Vector=true) {module=InitialMeshOrientation001}
   &InitialMeshOrientation(
    Mesh Orientation Mode=::NewEnumerator7
    Rotation=&SelectVectorFromArray(
     Direct Array Index=$Particles.PartitionParticles.Partition
     Vector Selection Array=DataInterface() {propertyValues {FloatData [<0 0 0>, <0 0 0.05>, <0 0 0>, <0 0 0.05>, <0 0 0>, <0 0 0.05>, <0 0 0>, <0 0 0.05>, <0 0 0>, <0 0 0.05>]}}
    )
    Use Rotation Vector=true
   ) {module=InitialMeshOrientation002}
   &CameraOffset(Camera Offset Amount=-30)
   /SetVariables() {isSetParameters=true, module=SetVariables_#9}
  ]
  particleUpdateScript [
   /SetVariables(Particles.AlphaAll:NiagaraFloat=&Multiply_Float(
     A=1
     B=&FloatFromCurve(
      FloatCurve=DataInterface() {propertyValues {Curve {defaultValue=1, keys [{tangentMode=RCTM_User, time=0, value=0}, {tangentMode=RCTM_User, time=1, value=1}]}}}
      CurveIndex=$Emitter.Age
     )
    )) {isSetParameters=true, module=SetVariables_#10}
   &DynamicMaterialParameters(
    Index 0 Param 1=0.8
    Index 0 Param 4=&FloatFromCurve(
     FloatCurve=DataInterface() {propertyValues {Curve {defaultValue=1, keys [{tangentMode=RCTM_User, time=0, value=0}, {tangentMode=RCTM_User, time=1, value=1}]}}}
     CurveIndex=$Emitter.Age
    )
    Module.Index0Float0Write=true
    Module.Index0Float1Write=true
    Module.Index0Float2Write=true
    Module.Index0Float3Write=true
   )
   &ParticleState(Kill Particles When Lifetime Has Elapsed=false)
   &SkeletalMeshLocation(
    Bone / Socket Index=$Particles.PartitionParticles.Partition
    Bone Sampling Mode=::NewEnumerator4
    MaskSpawnGroup=false
    Sampled Position Offset=&AddVectorToPosition(
     Position=&SelectVectorFromArray(
      Direct Array Index=$Particles.PartitionParticles.Partition
      Vector Selection Array=DataInterface() {propertyValues {FloatData [<-20 0 0>, <20 0 0>, <-20 0 0>, <20 0 0>, <-20 0 0>, <20 0 0>, <-20 0 0>, <20 0 0>, <-20 0 0>, <20 0 0>]}}
     )
     Vector=<0 -30 -20>
    )
    Skeletal Mesh=$System.Skeletal Mesh Reader
   )
   &RegionMask(
    Offset:Vector3f=$System.RegionMaskOffset
    Region Falloff:NiagaraFloat=$System.RegionMaskFade
    Region Radius:NiagaraFloat=$System.RegionMaskRadius
    UseOffset=true
    UseRotation=false
   )
   &RegionMask(
    Offset:Vector3f=$System.RegionMaskOffset
    Region Falloff=0
    Region Radius:NiagaraFloat=$System.RegionMaskRadius
    UseOffset=true
    UseRotation=false
   ) {module=RegionMask001}
   &InterpolateOverTime(Rate Of Change=10, Target Float Value:NiagaraFloat=$Particles.RegionMask.Float)
   &ScaleMeshSize(Scale Factor=&Multiply_VectorByFloat(
     Vector=&Multiply_Vector(
      A=&SelectVectorFromArray(
       Direct Array Index=$Particles.PartitionParticles.Partition
       Vector Selection Array=DataInterface() {propertyValues {FloatData [<1 1 1>, <-1 1 1>, <1 1 1>, <-1 1 1>, <1 1 1>, <-1 1 1>, <1 1 1>, <-1 1 1>, <1 1 1>, <-1 1 1>]}}
      )
      B=<1 1 1>
     )
     Float=&Multiply_Float(
      A=0.7
      B=&MakeCustomFloatFromBool(
       Bool=&SetBoolByFloatComparison(
        A=&VectorLength(Vector=&Subtract_Vector(A=Linked() {linkedVariable {name=Particles.Position, type=NiagaraPosition}}, B=$Particles.Lastposition))
        B=100
       )
       True Float=0
       False Float=1
      )
     )
    ))
   /SetVariables(Particles.Lastposition:NiagaraPosition=$Particles.Position) {isSetParameters=true, module=SetVariables_#11}
   &ScaleColor(
    Scale Alpha=&Multiply_Float(
     A=&SelectFloatFromArray(
      Direct Array Index=$Particles.PartitionParticles.Partition
      Float Selection Array=DataInterface() {propertyValues {FloatData [1,1,1,1,1,1,1,1,1,1]}}
     )
     B=$Particles.InterpolateOverTime.InterpolateOverTime_Float.Moving Average
    )
    ScaleA=true
    ScaleRGB=true
   )
   &ScaleColor(
    Scale Alpha=&Multiply_Float(A=$Particles.RegionMask001.Float, B=$Particles.AlphaAll)
    ScaleA=true
    ScaleRGB=true
   ) {module=ScaleColor001}
   &FN_Mesh_Attributes(
    Color=&MakeLinearColorFromVectorAndFloat(
     "Vector (RGB)"=&Multiply_Vector(
      A=&MakeVectorFromLinearColorRGB(Color=&self:ScratchDynamicInput_01(
        BlotColor:LinearColor=<rgba 1 0.0146 0 0.4>
        CitadelColor:LinearColor=<rgba 0 0.5 1.5 0.4>
        FrothColor:LinearColor=<rgba 0 0 0 1>
        VariantColorIndex:NiagaraInt32=&MakeIntFromFloat(Float=$User.VariantColorIndex)
       ))
      B=<1 1 1>
     )
     "Float (Alpha)"=&Multiply_Float(
      A=&Multiply_Float(
       A=&Multiply_Float(
        A=&Multiply_Float(
         A=0.5
         B=&self:ScratchDynamicInput_10(
          Blot:NiagaraFloat=1
          Citadel:NiagaraFloat=$System.FECitadelGlow
          Froth:NiagaraFloat=&FN_FrontEndFloatModifier(Gameplay=3, Front End=20)
          VariantColorIndex:NiagaraInt32=&MakeIntFromFloat(Float=$User.VariantColorIndex)
         )
        )
        B=&Multiply_Float(A=&FN_FrontEndFloatModifier(Gameplay=1, Front End=0.2), B=$Particles.AlphaAll)
       )
       B=$Particles.InterpolateOverTime.InterpolateOverTime_Float.Moving Average
      )
      B=&Multiply_Float(A=$Particles.RegionMask001.Float, B=$Particles.AlphaAll)
     )
    )
    Create Color Attribute=true
    Module Output=::NewEnumerator1
   ) {module=FN_Mesh_Attributes001}
   &FN_Mesh_Attributes(
    Color=&MakeLinearColorFromVectorAndFloat(
     "Vector (RGB)"=&Multiply_Vector(
      A=&MakeVectorFromLinearColorRGB(Color=&self:ScratchDynamicInput_01(
        BlotColor:LinearColor=<rgba 1 0.0146 0 0.4>
        CitadelColor:LinearColor=<rgba 0 0.5 1.5 0.4>
        FrothColor:LinearColor=<rgba 0 0 0 1>
        VariantColorIndex:NiagaraInt32=&MakeIntFromFloat(Float=$User.VariantColorIndex)
       ))
      B=<1 1 1>
     )
     "Float (Alpha)"=&Multiply_Float(
      A=&Multiply_Float(
       A=&Multiply_Float(
        A=&Multiply_Float(
         A=0.1
         B=&self:ScratchDynamicInput_10(
          Blot:NiagaraFloat=1
          Citadel:NiagaraFloat=$System.FECitadelGlow
          Froth:NiagaraFloat=&FN_FrontEndFloatModifier(Gameplay=3, Front End=20)
          VariantColorIndex:NiagaraInt32=&MakeIntFromFloat(Float=$User.VariantColorIndex)
         )
        )
        B=&Multiply_Float(A=&FN_FrontEndFloatModifier(Gameplay=1, Front End=0.2), B=$Particles.AlphaAll)
       )
       B=$Particles.InterpolateOverTime.InterpolateOverTime_Float.Moving Average
      )
      B=&Multiply_Float(A=$Particles.RegionMask001.Float, B=$Particles.AlphaAll)
     )
    )
    Create Color Attribute=true
    Module Output=::NewEnumerator1
   ) {module=FN_Mesh_Attributes002}
  ]
  rendererClasses [&NiagaraMeshRendererProperties]
  renderers [
   {rendererClass=&NiagaraMeshRendererProperties, rendererIndex=0}
   {rendererClass=&NiagaraMeshRendererProperties, rendererIndex=1}
   {rendererClass=&NiagaraMeshRendererProperties, rendererIndex=2}
  ]
 }
 IdleTrailOutSide {
  emitterUpdateScript [
   /SetVariables(Emitter.Des:NiagaraFloat=&self:ScratchDynamicInput_08(
     Blot:NiagaraFloat=0.5
     Citadel:NiagaraFloat=0.5
     Froth:NiagaraFloat=1
     VariantColorIndex:NiagaraInt32=&MakeIntFromFloat(Float=$User.VariantColorIndex)
    )) {isSetParameters=true, module=SetVariables_#12}
   &EmitterState(Life Cycle Mode=::NewEnumerator1, Loop Behavior=::NewEnumerator1, Loop Duration Mode=::NewEnumerator1, UseLoopDelay=true)
   &SpawnBurst_Instantaneous(Spawn Count=10, Use Spawn Probability=false, UseLoopCountLimit=false)
  ]
  particleSpawnScript [
   &PartitionParticles(Number Of Partitions=10)
   &InitializeParticle(Color=<rgba 2 2 2 2>, Mesh Scale=<1.2 1.2 1>, Mesh Scale Mode=::NewEnumerator4, UsePositionOffset=false, Write Material Random=true)
   &InitialMeshOrientation(
    Mesh Orientation Mode=::NewEnumerator7
    Rotation=&FN_FrontEndVector3Modifier(Gameplay=<0 0 0>, Front End=<0.03 0 0>)
    Use Rotation Vector=true
   )
   &InitialMeshOrientation(
    Mesh Orientation Mode=::NewEnumerator7
    Rotation=&SelectVectorFromArray(
     Direct Array Index=$Particles.PartitionParticles.Partition
     Vector Selection Array=DataInterface() {propertyValues {FloatData [<0 0 0>, <0 0 0.1>, <0 0 0>, <0 0 0.1>, <0 0 0>, <0 0 0.1>, <0 0 0>, <0 0 0.1>, <0 0 0>, <0 0 0.1>]}}
    )
    Use Rotation Vector=true
   ) {module=InitialMeshOrientation002}
   &InitialMeshOrientation(
    Mesh Orientation Mode=::NewEnumerator7
    Rotation=&SelectVectorFromArray(
     Direct Array Index=$Particles.PartitionParticles.Partition
     Vector Selection Array=DataInterface() {propertyValues {FloatData [<0 0 0>, <0 0.5 0>, <0 0 0>, <0 0.5 0>, <0 0 0>, <0 0.5 0>, <0 0 0>, <0 0.5 0>, <0 0 0>, <0 0.5 0>]}}
    )
    Use Rotation Vector=true
   ) {module=InitialMeshOrientation001}
   &CameraOffset(Camera Offset Amount=-30)
   /SetVariables() {isSetParameters=true, module=SetVariables_#9}
  ]
  particleUpdateScript [
   /SetVariables(Particles.AlphaAll:NiagaraFloat=&Multiply_Float(
     A=1
     B=&FloatFromCurve(
      FloatCurve=DataInterface() {propertyValues {Curve {defaultValue=1, keys [{tangentMode=RCTM_User, time=0, value=0}, {tangentMode=RCTM_User, time=1, value=1}]}}}
      CurveIndex=$Emitter.Age
     )
    )) {isSetParameters=true, module=SetVariables_#10}
   &DynamicMaterialParameters(
    Index 0 Param 1=&RandomRangeFloat(Minimum=1.3, Maximum=1.7)
    Index 0 Param 4=&FloatFromCurve(
     FloatCurve=DataInterface() {propertyValues {Curve {defaultValue=1, keys [{tangentMode=RCTM_User, time=0, value=0}, {tangentMode=RCTM_User, time=1, value=1}]}}}
     CurveIndex=$Emitter.Age
    )
    Module.Index0Float0Write=true
    Module.Index0Float1Write=true
    Module.Index0Float2Write=true
    Module.Index0Float3Write=true
   )
   &ParticleState(Kill Particles When Lifetime Has Elapsed=false)
   &SkeletalMeshLocation(
    Bone / Socket Index=$Particles.PartitionParticles.Partition
    Bone Sampling Mode=::NewEnumerator4
    MaskSpawnGroup=false
    Sampled Position Offset=&AddVectorToPosition(
     Position=&SelectVectorFromArray(
      Direct Array Index=$Particles.PartitionParticles.Partition
      Vector Selection Array=DataInterface() {propertyValues {FloatData [<-20 0 0>, <20 0 0>, <-20 0 0>, <20 0 0>, <-20 0 0>, <20 0 0>, <-20 0 0>, <20 0 0>, <-20 0 0>, <20 0 0>]}}
     )
     Vector=&FN_FrontEndVector3Modifier(Gameplay=<0 50 -15>, Front End=<0 50 0>)
    )
    Skeletal Mesh=$System.Skeletal Mesh Reader
   )
   &RegionMask(
    Offset:Vector3f=$System.RegionMaskOffset
    Region Falloff:NiagaraFloat=$System.RegionMaskFade
    Region Radius:NiagaraFloat=$System.RegionMaskRadius
    UseOffset=true
    UseRotation=false
   )
   &RegionMask(
    Offset:Vector3f=$System.RegionMaskOffset
    Region Falloff=0
    Region Radius:NiagaraFloat=$System.RegionMaskRadius
    UseOffset=true
    UseRotation=false
   ) {module=RegionMask001}
   &InterpolateOverTime(Rate Of Change=10, Target Float Value:NiagaraFloat=$Particles.RegionMask.Float)
   &ScaleMeshSize(Scale Factor=&Multiply_VectorByFloat(
     Vector=&Multiply_Vector(
      A=&SelectVectorFromArray(
       Direct Array Index=$Particles.PartitionParticles.Partition
       Vector Selection Array=DataInterface() {propertyValues {FloatData [<1 1 1>, <1 1 1>, <1 1 1>, <1 1 1>, <1 1 1>, <1 1 1>, <1 1 1>, <1 1 1>, <1 1 1>, <1 1 1>]}}
      )
      B=<1.5 1.5 1>
     )
     Float=&Multiply_Float(
      A=&RandomRangeFloat(Minimum=0.8, Maximum=0.9)
      B=&MakeCustomFloatFromBool(
       Bool=&SetBoolByFloatComparison(
        A=&VectorLength(Vector=&Subtract_Vector(A=Linked() {linkedVariable {name=Particles.Position, type=NiagaraPosition}}, B=$Particles.Lastposition))
        B=100
       )
       True Float=0
       False Float=1
      )
     )
    ))
   /SetVariables(Particles.Lastposition:NiagaraPosition=$Particles.Position) {isSetParameters=true, module=SetVariables_#11}
   &ScaleColor(
    Scale Alpha=&Multiply_Float(
     A=&SelectFloatFromArray(
      Direct Array Index=$Particles.PartitionParticles.Partition
      Float Selection Array=DataInterface() {propertyValues {FloatData [1,1,1,1,1,1,1,1,1,1]}}
     )
     B=$Particles.InterpolateOverTime.InterpolateOverTime_Float.Moving Average
    )
    ScaleA=true
    ScaleRGB=true
   )
   &ScaleColor(
    Scale Alpha=&Multiply_Float(A=$Particles.RegionMask001.Float, B=$Particles.AlphaAll)
    ScaleA=true
    ScaleRGB=true
   ) {module=ScaleColor001}
   &FN_Mesh_Attributes(
    Color=&MakeLinearColorFromVectorAndFloat(
     "Vector (RGB)"=&Multiply_Vector(
      A=&MakeVectorFromLinearColorRGB(Color=&self:ScratchDynamicInput_01(
        BlotColor:LinearColor=<rgba 1 0.0146 0 0.4>
        CitadelColor:LinearColor=<rgba 0 0.5 1.5 0.4>
        FrothColor:LinearColor=<rgba 0 0 0 1>
        VariantColorIndex:NiagaraInt32=&MakeIntFromFloat(Float=$User.VariantColorIndex)
       ))
      B=<1 1 1>
     )
     "Float (Alpha)"=&Multiply_Float(
      A=&Multiply_Float(
       A=&Multiply_Float(
        A=&Multiply_Float(
         A=0.5
         B=&self:ScratchDynamicInput_10(
          Blot:NiagaraFloat=1
          Citadel:NiagaraFloat=$System.FECitadelGlow
          Froth:NiagaraFloat=&FN_FrontEndFloatModifier(Gameplay=3, Front End=20)
          VariantColorIndex:NiagaraInt32=&MakeIntFromFloat(Float=$User.VariantColorIndex)
         )
        )
        B=&Multiply_Float(A=&FN_FrontEndFloatModifier(Gameplay=1, Front End=0.2), B=$Particles.AlphaAll)
       )
       B=$Particles.InterpolateOverTime.InterpolateOverTime_Float.Moving Average
      )
      B=&Multiply_Float(A=$Particles.RegionMask001.Float, B=$Particles.AlphaAll)
     )
    )
    Create Color Attribute=true
    Module Output=::NewEnumerator1
   ) {module=FN_Mesh_Attributes001}
   &FN_Mesh_Attributes(
    Color=&MakeLinearColorFromVectorAndFloat(
     "Vector (RGB)"=&Multiply_Vector(
      A=&MakeVectorFromLinearColorRGB(Color=&self:ScratchDynamicInput_01(
        BlotColor:LinearColor=<rgba 1 0.0146 0 0.4>
        CitadelColor:LinearColor=<rgba 0 0.5 1.5 0.4>
        FrothColor:LinearColor=<rgba 0 0 0 1>
        VariantColorIndex:NiagaraInt32=&MakeIntFromFloat(Float=$User.VariantColorIndex)
       ))
      B=<1 1 1>
     )
     "Float (Alpha)"=&Multiply_Float(
      A=&Multiply_Float(
       A=&Multiply_Float(
        A=&Multiply_Float(
         A=0.1
         B=&self:ScratchDynamicInput_11(
          Blot:NiagaraFloat=1
          Citadel:NiagaraFloat=$System.FECitadelGlow
          Froth:NiagaraFloat=&FN_FrontEndFloatModifier(Gameplay=3, Front End=20)
          VariantColorIndex:NiagaraInt32=&MakeIntFromFloat(Float=$User.VariantColorIndex)
         )
        )
        B=&Multiply_Float(A=&FN_FrontEndFloatModifier(Gameplay=1, Front End=0.2), B=$Particles.AlphaAll)
       )
       B=$Particles.InterpolateOverTime.InterpolateOverTime_Float.Moving Average
      )
      B=&Multiply_Float(A=$Particles.RegionMask001.Float, B=$Particles.AlphaAll)
     )
    )
    Create Color Attribute=true
    Module Output=::NewEnumerator1
   ) {module=FN_Mesh_Attributes002}
  ]
  rendererClasses [&NiagaraMeshRendererProperties]
  renderers [
   {rendererClass=&NiagaraMeshRendererProperties, rendererIndex=0}
   {rendererClass=&NiagaraMeshRendererProperties, rendererIndex=1}
   {rendererClass=&NiagaraMeshRendererProperties, rendererIndex=2}
  ]
 }
 IdleTrailOutSide001 {
  bEnabled=false
  emitterUpdateScript [
   &EmitterState(Life Cycle Mode=::NewEnumerator1, Loop Behavior=::NewEnumerator1, Loop Duration Mode=::NewEnumerator1, UseLoopDelay=true)
   &SpawnBurst_Instantaneous(Spawn Count=10, Use Spawn Probability=false, UseLoopCountLimit=false)
  ]
  particleSpawnScript [
   &PartitionParticles(Number Of Partitions=10)
   &InitializeParticle(Color=<rgba 2 2 2 2>, Mesh Scale=<1.2 1.2 1>, Mesh Scale Mode=::NewEnumerator4, UsePositionOffset=false, Write Material Random=true)
   &InitialMeshOrientation(
    Mesh Orientation Mode=::NewEnumerator7
    Rotation=&FN_FrontEndVector3Modifier(Gameplay=<0 0 0>, Front End=<0.03 0 0>)
    Use Rotation Vector=true
   )
   &InitialMeshOrientation(
    Mesh Orientation Mode=::NewEnumerator7
    Rotation=&SelectVectorFromArray(
     Direct Array Index=$Particles.PartitionParticles.Partition
     Vector Selection Array=DataInterface() {propertyValues {FloatData [<0 0 0>, <0 0 -0.1>, <0 0 0>, <0 0 -0.1>, <0 0 0>, <0 0 -0.1>, <0 0 0>, <0 0 -0.1>, <0 0 0>, <0 0 -0.1>]}}
    )
    Use Rotation Vector=true
   ) {module=InitialMeshOrientation002}
   &CameraOffset(Camera Offset Amount=-30)
   /SetVariables() {isSetParameters=true, module=SetVariables_#9}
  ]
  particleUpdateScript [
   /SetVariables(Particles.AlphaAll:NiagaraFloat=&Multiply_Float(
     A=1
     B=&FloatFromCurve(
      FloatCurve=DataInterface() {propertyValues {Curve {defaultValue=1, keys [{tangentMode=RCTM_User, time=0, value=0}, {tangentMode=RCTM_User, time=1, value=1}]}}}
      CurveIndex=$Emitter.Age
     )
    )) {isSetParameters=true, module=SetVariables_#10}
   &DynamicMaterialParameters(
    Index 0 Param 1=&RandomRangeFloat(Minimum=1.3, Maximum=1.7)
    Index 0 Param 4=&FloatFromCurve(
     FloatCurve=DataInterface() {propertyValues {Curve {defaultValue=1, keys [{tangentMode=RCTM_User, time=0, value=0}, {tangentMode=RCTM_User, time=1, value=1}]}}}
     CurveIndex=$Emitter.Age
    )
    Module.Index0Float0Write=true
    Module.Index0Float1Write=true
    Module.Index0Float2Write=true
    Module.Index0Float3Write=true
   )
   &ParticleState(Kill Particles When Lifetime Has Elapsed=false)
   &SkeletalMeshLocation(
    Bone / Socket Index=$Particles.PartitionParticles.Partition
    Bone Sampling Mode=::NewEnumerator4
    MaskSpawnGroup=false
    Sampled Position Offset=&AddVectorToPosition(
     Position=&SelectVectorFromArray(
      Direct Array Index=$Particles.PartitionParticles.Partition
      Vector Selection Array=DataInterface() {propertyValues {FloatData [<-20 0 0>, <20 0 0>, <-20 0 0>, <20 0 0>, <-20 0 0>, <20 0 0>, <-20 0 0>, <20 0 0>, <-20 0 0>, <20 0 0>]}}
     )
     Vector=&FN_FrontEndVector3Modifier(Gameplay=<0 50 -15>, Front End=<0 50 0>)
    )
    Skeletal Mesh=$System.Skeletal Mesh Reader
   )
   &RegionMask(
    Offset:Vector3f=$System.RegionMaskOffset
    Region Falloff:NiagaraFloat=$System.RegionMaskFade
    Region Radius:NiagaraFloat=$System.RegionMaskRadius
    UseOffset=true
    UseRotation=false
   )
   &RegionMask(
    Offset:Vector3f=$System.RegionMaskOffset
    Region Falloff=0
    Region Radius:NiagaraFloat=$System.RegionMaskRadius
    UseOffset=true
    UseRotation=false
   ) {module=RegionMask001}
   &InterpolateOverTime(Rate Of Change=10, Target Float Value:NiagaraFloat=$Particles.RegionMask.Float)
   &ScaleMeshSize(Scale Factor=&Multiply_VectorByFloat(
     Vector=&Multiply_Vector(
      A=&SelectVectorFromArray(
       Direct Array Index=$Particles.PartitionParticles.Partition
       Vector Selection Array=DataInterface() {propertyValues {FloatData [<1 1 1>, <-1 1 1>, <1 1 1>, <-1 1 1>, <1 1 1>, <-1 1 1>, <1 1 1>, <-1 1 1>, <1 1 1>, <-1 1 1>]}}
      )
      B=<1.5 1.5 1>
     )
     Float=&Multiply_Float(
      A=&RandomRangeFloat(Minimum=0.8, Maximum=0.9)
      B=&MakeCustomFloatFromBool(
       Bool=&SetBoolByFloatComparison(
        A=&VectorLength(Vector=&Subtract_Vector(A=Linked() {linkedVariable {name=Particles.Position, type=NiagaraPosition}}, B=$Particles.Lastposition))
        B=100
       )
       True Float=0
       False Float=1
      )
     )
    ))
   /SetVariables(Particles.Lastposition:NiagaraPosition=$Particles.Position) {isSetParameters=true, module=SetVariables_#11}
   &ScaleColor(
    Scale Alpha=&Multiply_Float(
     A=&SelectFloatFromArray(
      Direct Array Index=$Particles.PartitionParticles.Partition
      Float Selection Array=DataInterface() {propertyValues {FloatData [1,1,1,1,1,1,1,1,1,1]}}
     )
     B=$Particles.InterpolateOverTime.InterpolateOverTime_Float.Moving Average
    )
    ScaleA=true
    ScaleRGB=true
   )
   &ScaleColor(
    Scale Alpha=&Multiply_Float(A=$Particles.RegionMask001.Float, B=$Particles.AlphaAll)
    ScaleA=true
    ScaleRGB=true
   ) {module=ScaleColor001}
   &FN_Mesh_Attributes(
    Color=&MakeLinearColorFromVectorAndFloat(
     "Vector (RGB)"=&Multiply_Vector(A=&MakeVectorFromLinearColorRGB(Color=<rgba 1 0.0146 0 0.4>), B=<1 1 1>)
     "Float (Alpha)"=&Multiply_Float(
      A=&Multiply_Float(
       A=&Multiply_Float(
        A=0.5
        B=&Multiply_Float(A=&FN_FrontEndFloatModifier(Gameplay=1, Front End=0.2), B=$Particles.AlphaAll)
       )
       B=$Particles.InterpolateOverTime.InterpolateOverTime_Float.Moving Average
      )
      B=&Multiply_Float(A=$Particles.RegionMask001.Float, B=$Particles.AlphaAll)
     )
    )
    Create Color Attribute=true
    Module Output=::NewEnumerator1
   ) {module=FN_Mesh_Attributes001}
   &FN_Mesh_Attributes(
    Color=&MakeLinearColorFromVectorAndFloat(
     "Vector (RGB)"=&Multiply_Vector(A=&MakeVectorFromLinearColorRGB(Color=<rgba 1 0.0146 0 0.4>), B=<1 1 1>)
     "Float (Alpha)"=&Multiply_Float(
      A=&Multiply_Float(
       A=&Multiply_Float(
        A=0.1
        B=&Multiply_Float(A=&FN_FrontEndFloatModifier(Gameplay=1, Front End=0.2), B=$Particles.AlphaAll)
       )
       B=$Particles.InterpolateOverTime.InterpolateOverTime_Float.Moving Average
      )
      B=&Multiply_Float(A=$Particles.RegionMask001.Float, B=$Particles.AlphaAll)
     )
    )
    Create Color Attribute=true
    Module Output=::NewEnumerator1
   ) {module=FN_Mesh_Attributes002}
  ]
  rendererClasses [&NiagaraMeshRendererProperties]
  renderers [
   {rendererClass=&NiagaraMeshRendererProperties, rendererIndex=0}
   {rendererClass=&NiagaraMeshRendererProperties, rendererIndex=1}
   {rendererClass=&NiagaraMeshRendererProperties, rendererIndex=2}
  ]
 }
 IdleTrailOutSide002 {
  bEnabled=false
  emitterUpdateScript [
   /SetVariables(Emitter.Des:NiagaraFloat=&self:ScratchDynamicInput_08(
     Blot:NiagaraFloat=0.5
     Citadel:NiagaraFloat=0.5
     Froth:NiagaraFloat=1
     VariantColorIndex:NiagaraInt32=&MakeIntFromFloat(Float=$User.VariantColorIndex)
    )) {isSetParameters=true, module=SetVariables_#12}
   &EmitterState(Life Cycle Mode=::NewEnumerator1, Loop Behavior=::NewEnumerator1, Loop Duration Mode=::NewEnumerator1, UseLoopDelay=true)
   &SpawnBurst_Instantaneous(Spawn Count=10, Use Spawn Probability=false, UseLoopCountLimit=false)
  ]
  particleSpawnScript [
   &PartitionParticles(Number Of Partitions=10)
   &InitializeParticle(Color=<rgba 2 2 2 2>, Mesh Scale=<1.2 1.2 1>, Mesh Scale Mode=::NewEnumerator4, UsePositionOffset=false, Write Material Random=true)
   &InitialMeshOrientation(
    Mesh Orientation Mode=::NewEnumerator7
    Rotation=&FN_FrontEndVector3Modifier(Gameplay=<0 0 0>, Front End=<0.03 0 0>)
    Use Rotation Vector=true
   )
   &InitialMeshOrientation(
    Mesh Orientation Mode=::NewEnumerator7
    Rotation=&SelectVectorFromArray(
     Direct Array Index=$Particles.PartitionParticles.Partition
     Vector Selection Array=DataInterface() {propertyValues {FloatData [<0 0 0>, <0 0 -0.1>, <0 0 0>, <0 0 -0.1>, <0 0 0>, <0 0 -0.1>, <0 0 0>, <0 0 -0.1>, <0 0 0>, <0 0 -0.1>]}}
    )
    Use Rotation Vector=true
   ) {module=InitialMeshOrientation002}
   &CameraOffset(Camera Offset Amount=-30)
   /SetVariables() {isSetParameters=true, module=SetVariables_#9}
  ]
  particleUpdateScript [
   /SetVariables(Particles.AlphaAll:NiagaraFloat=&Multiply_Float(
     A=1
     B=&FloatFromCurve(
      FloatCurve=DataInterface() {propertyValues {Curve {defaultValue=1, keys [{tangentMode=RCTM_User, time=0, value=0}, {tangentMode=RCTM_User, time=1, value=1}]}}}
      CurveIndex=$Emitter.Age
     )
    )) {isSetParameters=true, module=SetVariables_#10}
   &DynamicMaterialParameters(
    Index 0 Param 1=&RandomRangeFloat(Minimum=1.3, Maximum=1.7)
    Index 0 Param 4=&FloatFromCurve(
     FloatCurve=DataInterface() {propertyValues {Curve {defaultValue=1, keys [{tangentMode=RCTM_User, time=0, value=0}, {tangentMode=RCTM_User, time=1, value=1}]}}}
     CurveIndex=$Emitter.Age
    )
    Module.Index0Float0Write=true
    Module.Index0Float1Write=true
    Module.Index0Float2Write=true
    Module.Index0Float3Write=true
   )
   &ParticleState(Kill Particles When Lifetime Has Elapsed=false)
   &SkeletalMeshLocation(
    Bone / Socket Index=$Particles.PartitionParticles.Partition
    Bone Sampling Mode=::NewEnumerator4
    MaskSpawnGroup=false
    Sampled Position Offset=&AddVectorToPosition(
     Position=&SelectVectorFromArray(
      Direct Array Index=$Particles.PartitionParticles.Partition
      Vector Selection Array=DataInterface() {propertyValues {FloatData [<-20 0 0>, <20 0 0>, <-20 0 0>, <20 0 0>, <-20 0 0>, <20 0 0>, <-20 0 0>, <20 0 0>, <-20 0 0>, <20 0 0>]}}
     )
     Vector=&FN_FrontEndVector3Modifier(Gameplay=<0 50 -15>, Front End=<0 50 0>)
    )
    Skeletal Mesh=$System.Skeletal Mesh Reader
   )
   &RegionMask(
    Offset:Vector3f=$System.RegionMaskOffset
    Region Falloff:NiagaraFloat=$System.RegionMaskFade
    Region Radius:NiagaraFloat=$System.RegionMaskRadius
    UseOffset=true
    UseRotation=false
   )
   &RegionMask(
    Offset:Vector3f=$System.RegionMaskOffset
    Region Falloff=0
    Region Radius:NiagaraFloat=$System.RegionMaskRadius
    UseOffset=true
    UseRotation=false
   ) {module=RegionMask001}
   &InterpolateOverTime(Rate Of Change=10, Target Float Value:NiagaraFloat=$Particles.RegionMask.Float)
   &ScaleMeshSize(Scale Factor=&Multiply_VectorByFloat(
     Vector=&Multiply_Vector(
      A=&SelectVectorFromArray(
       Direct Array Index=$Particles.PartitionParticles.Partition
       Vector Selection Array=DataInterface() {propertyValues {FloatData [<1 1 1>, <-1 1 1>, <1 1 1>, <-1 1 1>, <1 1 1>, <-1 1 1>, <1 1 1>, <-1 1 1>, <1 1 1>, <-1 1 1>]}}
      )
      B=<1.5 1.5 1>
     )
     Float=&Multiply_Float(
      A=&RandomRangeFloat(Minimum=0.8, Maximum=0.9)
      B=&MakeCustomFloatFromBool(
       Bool=&SetBoolByFloatComparison(
        A=&VectorLength(Vector=&Subtract_Vector(A=Linked() {linkedVariable {name=Particles.Position, type=NiagaraPosition}}, B=$Particles.Lastposition))
        B=100
       )
       True Float=0
       False Float=1
      )
     )
    ))
   /SetVariables(Particles.Lastposition:NiagaraPosition=$Particles.Position) {isSetParameters=true, module=SetVariables_#11}
   &ScaleColor(
    Scale Alpha=&Multiply_Float(
     A=&SelectFloatFromArray(
      Direct Array Index=$Particles.PartitionParticles.Partition
      Float Selection Array=DataInterface() {propertyValues {FloatData [1,1,1,1,1,1,1,1,1,1]}}
     )
     B=$Particles.InterpolateOverTime.InterpolateOverTime_Float.Moving Average
    )
    ScaleA=true
    ScaleRGB=true
   )
   &ScaleColor(
    Scale Alpha=&Multiply_Float(A=$Particles.RegionMask001.Float, B=$Particles.AlphaAll)
    ScaleA=true
    ScaleRGB=true
   ) {module=ScaleColor001}
   &FN_Mesh_Attributes(
    Color=&MakeLinearColorFromVectorAndFloat(
     "Vector (RGB)"=&Multiply_Vector(
      A=&MakeVectorFromLinearColorRGB(Color=&self:ScratchDynamicInput_01(
        BlotColor:LinearColor=<rgba 1 0.0146 0 0.4>
        CitadelColor:LinearColor=<rgba 0 0.5 1.5 0.4>
        FrothColor:LinearColor=<rgba 0 0 0 1>
        VariantColorIndex:NiagaraInt32=&MakeIntFromFloat(Float=$User.VariantColorIndex)
       ))
      B=<1 1 1>
     )
     "Float (Alpha)"=&Multiply_Float(
      A=&Multiply_Float(
       A=&Multiply_Float(
        A=&Multiply_Float(
         A=0.5
         B=&self:ScratchDynamicInput_10(
          Blot:NiagaraFloat=1
          Citadel:NiagaraFloat=$System.FECitadelGlow
          Froth:NiagaraFloat=&FN_FrontEndFloatModifier(Gameplay=3, Front End=20)
          VariantColorIndex:NiagaraInt32=&MakeIntFromFloat(Float=$User.VariantColorIndex)
         )
        )
        B=&Multiply_Float(A=&FN_FrontEndFloatModifier(Gameplay=1, Front End=0.2), B=$Particles.AlphaAll)
       )
       B=$Particles.InterpolateOverTime.InterpolateOverTime_Float.Moving Average
      )
      B=&Multiply_Float(A=$Particles.RegionMask001.Float, B=$Particles.AlphaAll)
     )
    )
    Create Color Attribute=true
    Module Output=::NewEnumerator1
   ) {module=FN_Mesh_Attributes001}
   &FN_Mesh_Attributes(
    Color=&MakeLinearColorFromVectorAndFloat(
     "Vector (RGB)"=&Multiply_Vector(
      A=&MakeVectorFromLinearColorRGB(Color=&self:ScratchDynamicInput_01(
        BlotColor:LinearColor=<rgba 1 0.0146 0 0.4>
        CitadelColor:LinearColor=<rgba 0 0.5 1.5 0.4>
        FrothColor:LinearColor=<rgba 0 0 0 1>
        VariantColorIndex:NiagaraInt32=&MakeIntFromFloat(Float=$User.VariantColorIndex)
       ))
      B=<1 1 1>
     )
     "Float (Alpha)"=&Multiply_Float(
      A=&Multiply_Float(
       A=&Multiply_Float(
        A=&Multiply_Float(
         A=0.1
         B=&self:ScratchDynamicInput_11(
          Blot:NiagaraFloat=1
          Citadel:NiagaraFloat=$System.FECitadelGlow
          Froth:NiagaraFloat=&FN_FrontEndFloatModifier(Gameplay=3, Front End=20)
          VariantColorIndex:NiagaraInt32=&MakeIntFromFloat(Float=$User.VariantColorIndex)
         )
        )
        B=&Multiply_Float(A=&FN_FrontEndFloatModifier(Gameplay=1, Front End=0.2), B=$Particles.AlphaAll)
       )
       B=$Particles.InterpolateOverTime.InterpolateOverTime_Float.Moving Average
      )
      B=&Multiply_Float(A=$Particles.RegionMask001.Float, B=$Particles.AlphaAll)
     )
    )
    Create Color Attribute=true
    Module Output=::NewEnumerator1
   ) {module=FN_Mesh_Attributes002}
  ]
  rendererClasses [&NiagaraMeshRendererProperties]
  renderers [
   {rendererClass=&NiagaraMeshRendererProperties, rendererIndex=0}
   {rendererClass=&NiagaraMeshRendererProperties, rendererIndex=1}
   {rendererClass=&NiagaraMeshRendererProperties, rendererIndex=2}
  ]
 }
 Lightning {
  emitterUpdateScript [
   /SetVariables(Emitter.BackGroundColor:LinearColor=&self:ScratchDynamicInput_04(
     BlotColor:LinearColor=<rgba 0.5 0.025 0.025 1>
     CitadelColor:LinearColor=<rgba 0 0.0667 0.2 1>
     FrothColor:LinearColor=<rgba 0 0 0 1>
     VariantColorIndex:NiagaraInt32=&MakeIntFromFloat(Float=$User.VariantColorIndex)
    )) {isSetParameters=true}
   /SetVariables(Emitter.CenterColorNoise:LinearColor=&self:ScratchDynamicInput_02(
     BlotColor:LinearColor=<rgba 7 0.35 0.35 1>
     CitadelColor:LinearColor=<rgba 0 0 0 1>
     FrothColor:LinearColor=<rgba 0.5 0.5 0.5 1>
     VariantColorIndex:NiagaraInt32=&MakeIntFromFloat(Float=$User.VariantColorIndex)
    )) {isSetParameters=true}
   /SetVariables(Emitter.OutLineColor:LinearColor=&self:ScratchDynamicInput_03(
     BlotColor:LinearColor=<rgba 0.3 0 0.01 1>
     CitadelColor:LinearColor=<rgba 0 1 2 1>
     FrothColor:LinearColor=<rgba 0.05 0.05 0.05 1>
     VariantColorIndex:NiagaraInt32=&MakeIntFromFloat(Float=$User.VariantColorIndex)
    )) {isSetParameters=true}
   /SetVariables(Emitter.Des:NiagaraFloat=&self:ScratchDynamicInput_08(
     Blot:NiagaraFloat=0.5
     Citadel:NiagaraFloat=0.5
     Froth:NiagaraFloat=1
     VariantColorIndex:NiagaraInt32=&MakeIntFromFloat(Float=$User.VariantColorIndex)
    )) {isSetParameters=true}
   &EmitterState()
   &SpawnRate(
    Spawn Probability=&FN_FrontEndFloatModifier(Gameplay=$User.ForwardDot, Front End=1)
    SpawnRate=15
    Use Spawn Probability=true
   )
  ]
  particleSpawnScript [
   &InitializeParticle(
    Color=<rgba 5 5 5 1>
    Lifetime Max=1
    Lifetime Min=0.6
    Lifetime Mode=::NewEnumerator1
    Sprite Rotation Mode=::NewEnumerator2
    Sprite Size Max=<20 30>
    Sprite Size Min=<10 15>
    Sprite Size Mode=::NewEnumerator2
    UsePositionOffset=false
    Write Material Random=true
   )
   &SubUVAnimation(
    Evaluation Type=::NewEnumerator0
    "Sprite Renderer "=DataInterface() {propertyValues {SpriteRenderer=&self:Lightning_0.NiagaraSpriteRendererProperties_0}}
    SubUV Animation Mode:ENiagara_SubUVLookupModeV2=::NewEnumerator2
    UseEndFrame=false
    UseStartFrame=false
   )
   &ShapeLocation(
    Handle Radius=0
    Large Radius=30
    MaskSpawnGroup=false
    Offset Mode=::NewEnumerator0
    Override Seed=false
    Shape Primitive=::NewEnumerator3
    UseOwnerScale=true
   )
   &AddVelocity(Module.Use Falloff Distance=false, Origin Offset=<0 0 -30>, Override Seed=false, Velocity Mode=::NewEnumerator1, Velocity Speed=200)
  ]
  particleUpdateScript [
   &ParticleState()
   &ScaleSpriteSize(Uniform Curve Sprite Scale=DataInterface() {propertyValues {Curve {
       defaultValue=0
       keys [{interpMode=RCIM_Cubic, tangentMode=RCTM_User, time=0, value=1.8}, {interpMode=RCIM_Cubic, tangentMode=RCTM_User, time=1, value=0.5}]
      }}})
   &SpriteRotationRate(Rotation Rate=&RandomRangeFloat(Minimum=-200, Maximum=200))
   &Drag(Drag=2, Use Linear Drag=true, Use Rotational Drag=true)
   &GravityForce(Gravity=<0 0 200>)
   &CurlNoiseForce(Bias Noise=false, Noise Frequency=100, Noise Strength=500, Pan Noise=false)
   &SolveForcesAndVelocity(Clamp Velocity=false, Limit Acceleration=false)
   &FN_Mesh_Attributes(
    Color=&MakeLinearColorFromVectorAndFloat(
     "Vector (RGB)"=&Multiply_Vector(
      A=&MakeVectorFromLinearColorRGB(Color=&self:ScratchDynamicInput_04(
        BlotColor:LinearColor=<rgba 1 0.0146 0 0.4>
        CitadelColor:LinearColor=<rgba 0 0.3333 1 0.4>
        FrothColor:LinearColor=<rgba 0 0 0 1>
        VariantColorIndex:NiagaraInt32=&MakeIntFromFloat(Float=$User.VariantColorIndex)
       ))
      B=&Multiply_VectorByFloat(Vector=<1 1 1>, Float=2)
     )
     "Float (Alpha)"=&Multiply_Float(A=4, B=1)
    )
    Create Color Attribute=true
    Module Output=::NewEnumerator1
   )
   &DynamicMaterialParameters(
    Index 0 Param 4=&FloatFromCurve(FloatCurve=DataInterface() {propertyValues {Curve {defaultValue=1, keys [{tangentMode=RCTM_User, time=0, value=0.3}, {tangentMode=RCTM_User, time=1, value=1}]}}})
    Module.Index0Float0Write=true
    Module.Index0Float1Write=true
    Module.Index0Float2Write=true
    Module.Index0Float3Write=true
   )
   &ScaleColor(
    Scale RGB=&Multiply_VectorByFloat(
     Vector=&Multiply_VectorByFloat(Vector=<1 1 1>, Float=1)
     Float=&FloatFromCurve(FloatCurve=DataInterface() {propertyValues {Curve {keys [
          <at 0 20 RCIM_Cubic>
          {arriveTangent=-28.4647, interpMode=RCIM_Cubic, leaveTangent=-28.4647, time=0.2677, value=15.1473}
          {arriveTangent=-17.9534, interpMode=RCIM_Cubic, leaveTangent=-17.9534, time=0.5865, value=3.3041}
          <at 1 2>
         ]}}})
    )
    ScaleA=true
    ScaleRGB=true
   )
  ]
  rendererClasses [&NiagaraSpriteRendererProperties]
  renderers [
   {rendererClass=&NiagaraSpriteRendererProperties}
   {rendererClass=&NiagaraSpriteRendererProperties, rendererIndex=1}
  ]
 }
 SmokeBlur {
  emitterUpdateScript [
   &EmitterState(Life Cycle Mode=::NewEnumerator1, Loop Duration=1, UseLoopDelay=false)
   &SpawnRate(
    Spawn Probability=&FN_FrontEndFloatModifier(Gameplay=$User.ForwardDot, Front End=1)
    SpawnRate=16
    Use Spawn Probability=true
   )
   &SpawnBurst_Instantaneous(
    Spawn Count=4
    Spawn Probability=&FN_FrontEndFloatModifier(Gameplay=$User.ForwardDot, Front End=1)
    Use Spawn Probability=true
    UseLoopCountLimit=false
   )
   /SetVariables(Emitter.BlurSmokeRender=true) {enabled=false, isSetParameters=true, module=SetVariables_#13}
   /SetVariables(Emitter.DetailSmokeRender=true) {enabled=false, isSetParameters=true, module=SetVariables_#14}
  ]
  particleSpawnScript [
   &PartitionParticles()
   &InitializeParticle(
    Color Mode=::NewEnumerator0
    Lifetime Max=1.5
    Lifetime Mode=::NewEnumerator1
    Mesh Scale=<1.2 1.2 1>
    Mesh Scale Mode=::NewEnumerator4
    Sprite Rotation Mode=::NewEnumerator2
    Sprite Size Mode=::NewEnumerator1
    Uniform Sprite Size Max=200
    Uniform Sprite Size Min=150
    UsePositionOffset=false
    Write Material Random=true
   )
   &CameraOffset(Camera Offset Amount=-30)
   &AddVelocity(Velocity=&MakeVector(
     X=&Multiply_Float(
      A=&SelectFloatFromArray(Direct Array Index=$Particles.PartitionParticles.Partition, Float Selection Array=DataInterface() {propertyValues {FloatData [170,-170]}})
      B=&FN_FrontEndFloatModifier(Gameplay=1, Front End=0.8)
     )
     Y=200
     Z=0
    ))
   &SubUVAnimation(
    End Frame Range Override=4
    Evaluation Type=::NewEnumerator0
    "Sprite Renderer "=DataInterface() {propertyValues {SpriteRenderer=&self:SmokeBlur_0.NiagaraSpriteRendererProperties_0}}
    SubUV Animation Mode:ENiagara_SubUVLookupModeV2=::NewEnumerator2
    UseEndFrame=true
    UseStartFrame=true
   )
   &AddVelocity(Velocity=&FN_FrontEndVector3Modifier(Gameplay=<0 0 -50>, Front End=<0 0 25>)) {module=AddVelocity001}
  ]
  particleUpdateScript [
   &Color(Color=&self:ScratchDynamicInput_01(
     BlotColor:LinearColor=&LerpLinearColors(
      StartColor=<rgba 0.4 0 0.0333 1>
      EndColor=<rgba 0.3 0.0109 0.035 1>
      LerpFactor=&UniformAOrBFloat(A=-1, Distribution Weight=0.6, Evaluation Type=::NewEnumerator0, Recalculate Random Each Loop=true, Override Seed=false)
     )
     CitadelColor:LinearColor=&LerpLinearColors(
      StartColor=<rgba 0 0.1333 0.4 1>
      EndColor=<rgba 0.0073 0.0715 0.2 1>
      LerpFactor=&UniformAOrBFloat(A=-1, Distribution Weight=0.6, Evaluation Type=::NewEnumerator0, Recalculate Random Each Loop=true, Override Seed=false)
     )
     FrothColor:LinearColor=&LerpLinearColors(
      StartColor=<rgba 0 0 0 1>
      EndColor=<rgba 0 0 0 1>
      LerpFactor=&UniformAOrBFloat(A=-1, Distribution Weight=0.6, Evaluation Type=::NewEnumerator0, Recalculate Random Each Loop=true, Override Seed=false)
     )
     VariantColorIndex:NiagaraInt32=&MakeIntFromFloat(Float=$User.VariantColorIndex)
    ))
   &DynamicMaterialParameters(
    Index 0 Param 1=&RandomRangeFloat(Minimum=0, Maximum=0)
    Index 0 Param 2=&RandomRangeFloat(Minimum=0, Maximum=1)
    Module.Index0Float0Write=true
    Module.Index0Float1Write=true
    Module.Index0Float2Write=true
    Module.Index0Float3Write=true
    Module.Index1Float0Write=false
    Module.Index1Float1Write=false
    Module.Index1Float2Write=false
    Module.Index1Float3Write=false
    Write Parameter Index 1=::NewEnumerator1
   ) {module=DynamicMaterialParameters001}
   &ParticleState()
   &SpriteRotationRate(Rotation Rate=&RandomRangeFloat(Minimum=-50, Maximum=50))
   &ScaleSpriteSize(
    Scale Sprite Size Mode=::NewEnumerator0
    Uniform Scale Factor=&Multiply_Float(
     A=&FloatFromCurve(FloatCurve=DataInterface() {propertyValues {Curve {defaultValue=1, keys [{tangentMode=RCTM_User, time=0, value=0.5}, {tangentMode=RCTM_User, time=1, value=1}]}}})
     B=1
    )
   )
   &ScaleColor(
    Color Value To Scale=$Particles.Color
    Scale Alpha=&Multiply_Float(
     A=&Multiply_Float(
      A=&Multiply_Float(
       A=&FloatFromCurve(
        FloatCurve=DataInterface() {propertyValues {Curve {defaultValue=1, keys [{tangentMode=RCTM_User, time=0, value=0}, {tangentMode=RCTM_User, time=1, value=1}]}}}
        CurveIndex=$Emitter.Age
        Scale Curve=&FN_FrontEndFloatModifier(Gameplay=1, Front End=0.8)
       )
       B=&FloatFromCurve(
        FloatCurve=DataInterface() {propertyValues {Curve {
           defaultValue=1
           keys [
            {interpMode=RCIM_Cubic, tangentMode=RCTM_User, time=0, value=0}
            {interpMode=RCIM_Cubic, tangentMode=RCTM_User, time=0.5, value=1}
            {tangentMode=RCTM_User, time=1, value=0}
           ]
          }}}
        Scale Curve=0.7
       )
      )
      B=&self:ScratchDynamicInput_06(
       Blot:NiagaraFloat=1
       Citadel:NiagaraFloat=2
       Froth:NiagaraFloat=&FN_FrontEndFloatModifier(Gameplay=1.5, Front End=2)
       VariantColorIndex:NiagaraInt32=&MakeIntFromFloat(Float=$User.VariantColorIndex)
      )
     )
     B=&FN_PickFloatDependOnPlatform(TargetPlatformActive=1, TargetPlatformNoActive=0.7)
    )
    ScaleA=true
    ScaleRGB=true
   )
   &SolveForcesAndVelocity(Clamp Velocity=false, Limit Acceleration=false)
   &SubUVAnimation(
    End Frame Range Override=30
    Output Alternate Renderer Binding=true
    "Sprite Renderer "=DataInterface() {propertyValues {SpriteRenderer=&self:SmokeBlur_0.NiagaraSpriteRendererProperties_1}}
    Start Frame Range Override=&RandomRangeInt(Minimum=0, Maximum=5)
    UseEndFrame=true
    UseStartFrame=true
   ) {module=SubUVAnimation001}
   &FN_Sprite_Attributes(
    Color=&MakeLinearColorFromVectorAndFloat(
     "Vector (RGB)"=&Multiply_Vector(
      A=&MakeVectorFromLinearColorRGB(Color=&self:ScratchDynamicInput_01(
        BlotColor:LinearColor=&LerpLinearColors(
         StartColor=<rgba 0.5 0.025 0.0646 1.5>
         EndColor=<rgba 0.4 0.0144 0.0465 2>
         LerpFactor=&UniformAOrBFloat(A=-1, Evaluation Type=::NewEnumerator0, Recalculate Random Each Loop=true, Override Seed=false)
        )
        CitadelColor:LinearColor=<rgba 0.02 0.1467 0.4 1.5>
        FrothColor:LinearColor=<rgba 0 0 0 1>
        VariantColorIndex:NiagaraInt32=&MakeIntFromFloat(Float=$User.VariantColorIndex)
       ))
      B=&Multiply_VectorByFloat(Vector=<1 1 1>, Float=1.3)
     )
     "Float (Alpha)"=&Multiply_Float(
      A=&Multiply_Float(
       A=2
       B=&self:ScratchDynamicInput_06(
        Blot:NiagaraFloat=1
        Citadel:NiagaraFloat=&FN_FrontEndFloatModifier(Gameplay=2, Front End=3)
        Froth:NiagaraFloat=&FN_FrontEndFloatModifier(Gameplay=1.5, Front End=2.5)
        VariantColorIndex:NiagaraInt32=&MakeIntFromFloat(Float=$User.VariantColorIndex)
       )
      )
      B=&Multiply_Float(
       A=&FloatFromCurve(
        FloatCurve=DataInterface() {propertyValues {Curve {defaultValue=1, keys [{tangentMode=RCTM_User, time=0, value=0}, {tangentMode=RCTM_User, time=1, value=1}]}}}
        CurveIndex=$Emitter.Age
        Scale Curve=&FN_FrontEndFloatModifier(Gameplay=1, Front End=0.5)
       )
       B=&FloatFromCurve(
        FloatCurve=DataInterface() {propertyValues {Curve {
           defaultValue=1
           keys [
            {interpMode=RCIM_Cubic, tangentMode=RCTM_User, time=0, value=0}
            {interpMode=RCIM_Cubic, tangentMode=RCTM_User, time=0.5, value=1}
            {tangentMode=RCTM_User, time=1, value=0}
           ]
          }}}
        Scale Curve=&Multiply_Float(A=0.6, B=&FN_PickFloatDependOnPlatform(TargetPlatformActive=1, TargetPlatformNoActive=0.7))
       )
      )
     )
    )
    Create Color Attribute=true
    Create SpriteSize Attribute=true
    Module Output:ENiagara_AttributeSamplingApplyOutput=::NewEnumerator1
    SpriteSize=&Vector2DFromFloat(Value=&Multiply_Float(
      A=&RandomRangeFloat(Minimum=150, Maximum=200)
      B=&FloatFromCurve(
       FloatCurve=DataInterface() {propertyValues {Curve {defaultValue=1, keys [{tangentMode=RCTM_User, time=0, value=0.5}, {tangentMode=RCTM_User, time=1, value=1.3}]}}}
       Scale Curve=&FN_FrontEndFloatModifier(Gameplay=1, Front End=0.8)
      )
     ))
   )
   /SetVariables(Particles.VisibilityIndex:NiagaraInt32=&MakeIntFromFloat(Float=&UniformAOrBFloat(A=0, Distribution Weight=0.3, Evaluation Type=::NewEnumerator0, Recalculate Random Each Loop=true, Override Seed=false))) {isSetParameters=true, module=SetVariables_#15}
  ]
  rendererClasses [&NiagaraSpriteRendererProperties]
  renderers [
   {rendererClass=&NiagaraSpriteRendererProperties}
   {rendererClass=&NiagaraSpriteRendererProperties, rendererIndex=1}
  ]
 }
 SmokeBlur001 {
  bEnabled=false
  emitterUpdateScript [
   &EmitterState(Life Cycle Mode=::NewEnumerator1, Loop Duration=1, UseLoopDelay=false)
   &SpawnRate(
    Spawn Probability=&FN_FrontEndFloatModifier(Gameplay=$User.ForwardDot, Front End=1)
    SpawnRate=16
    Use Spawn Probability=false
   )
   &SpawnBurst_Instantaneous(
    Spawn Count=4
    Spawn Probability=&FN_FrontEndFloatModifier(Gameplay=$User.ForwardDot, Front End=1)
    Use Spawn Probability=false
    UseLoopCountLimit=false
   )
   /SetVariables(Emitter.BlurSmokeRender=true) {enabled=false, isSetParameters=true, module=SetVariables_#13}
   /SetVariables(Emitter.DetailSmokeRender=true) {enabled=false, isSetParameters=true, module=SetVariables_#14}
  ]
  particleSpawnScript [
   &PartitionParticles()
   &InitializeParticle(
    Color=&LerpLinearColors(
     StartColor=<rgba 0.4 0 0.0333 1>
     EndColor=<rgba 0.3 0.0109 0.035 1>
     LerpFactor=&UniformAOrBFloat(A=-1, Distribution Weight=0.6, Evaluation Type=::NewEnumerator0, Recalculate Random Each Loop=true, Override Seed=false)
    )
    Lifetime Max=1.5
    Lifetime Mode=::NewEnumerator1
    Mesh Scale=<1.2 1.2 1>
    Mesh Scale Mode=::NewEnumerator4
    Sprite Rotation Mode=::NewEnumerator2
    Sprite Size Mode=::NewEnumerator1
    Uniform Sprite Size Max=200
    Uniform Sprite Size Min=150
    UsePositionOffset=false
    Write Material Random=true
   )
   &CameraOffset(Camera Offset Amount=-30)
   &AddVelocity(Velocity=&MakeVector(
     X=&Multiply_Float(
      A=&SelectFloatFromArray(Direct Array Index=$Particles.PartitionParticles.Partition, Float Selection Array=DataInterface() {propertyValues {FloatData [170,-170]}})
      B=&FN_FrontEndFloatModifier(Gameplay=1, Front End=0.8)
     )
     Y=200
     Z=0
    ))
   &SubUVAnimation(
    End Frame Range Override=4
    Evaluation Type=::NewEnumerator0
    "Sprite Renderer "=DataInterface() {propertyValues {SpriteRenderer=&self:SmokeBlur001_0.NiagaraSpriteRendererProperties_0}}
    SubUV Animation Mode:ENiagara_SubUVLookupModeV2=::NewEnumerator2
    UseEndFrame=true
    UseStartFrame=true
   )
   &AddVelocity(Velocity=&FN_FrontEndVector3Modifier(Gameplay=<0 0 -50>, Front End=<0 0 25>)) {module=AddVelocity001}
  ]
  particleUpdateScript [
   &DynamicMaterialParameters(
    Index 0 Param 1=&RandomRangeFloat(Minimum=0, Maximum=0)
    Index 0 Param 2=&RandomRangeFloat(Minimum=0, Maximum=1)
    Module.Index0Float0Write=true
    Module.Index0Float1Write=true
    Module.Index0Float2Write=true
    Module.Index0Float3Write=true
    Module.Index1Float0Write=false
    Module.Index1Float1Write=false
    Module.Index1Float2Write=false
    Module.Index1Float3Write=false
    Write Parameter Index 1=::NewEnumerator1
   ) {module=DynamicMaterialParameters001}
   &ParticleState()
   &SpriteRotationRate(Rotation Rate=&RandomRangeFloat(Minimum=-50, Maximum=50))
   &ScaleSpriteSize(
    Scale Sprite Size Mode=::NewEnumerator0
    Uniform Scale Factor=&Multiply_Float(
     A=&FloatFromCurve(FloatCurve=DataInterface() {propertyValues {Curve {defaultValue=1, keys [{tangentMode=RCTM_User, time=0, value=0.5}, {tangentMode=RCTM_User, time=1, value=1}]}}})
     B=1
    )
   )
   &ScaleColor(
    Scale Alpha=&Multiply_Float(
     A=&FloatFromCurve(
      FloatCurve=DataInterface() {propertyValues {Curve {defaultValue=1, keys [{tangentMode=RCTM_User, time=0, value=0}, {tangentMode=RCTM_User, time=1, value=1}]}}}
      CurveIndex=$Emitter.Age
      Scale Curve=&FN_FrontEndFloatModifier(Gameplay=1, Front End=0.8)
     )
     B=&FloatFromCurve(
      FloatCurve=DataInterface() {propertyValues {Curve {
         defaultValue=1
         keys [
          {interpMode=RCIM_Cubic, tangentMode=RCTM_User, time=0, value=0}
          {interpMode=RCIM_Cubic, tangentMode=RCTM_User, time=0.5, value=1}
          {tangentMode=RCTM_User, time=1, value=0}
         ]
        }}}
      Scale Curve=0.7
     )
    )
    ScaleA=true
    ScaleRGB=true
   )
   &SolveForcesAndVelocity(Clamp Velocity=false, Limit Acceleration=false)
   &SubUVAnimation(
    End Frame Range Override=30
    Output Alternate Renderer Binding=true
    "Sprite Renderer "=DataInterface() {propertyValues {SpriteRenderer=&self:SmokeBlur001_0.NiagaraSpriteRendererProperties_1}}
    Start Frame Range Override=&RandomRangeInt(Minimum=0, Maximum=5)
    UseEndFrame=true
    UseStartFrame=true
   ) {module=SubUVAnimation001}
   &FN_Sprite_Attributes(
    Color=&MakeLinearColorFromVectorAndFloat(
     "Vector (RGB)"=&Multiply_Vector(
      A=&MakeVectorFromLinearColorRGB(Color=&LerpLinearColors(
        StartColor=<rgba 0.5 0.025 0.0646 1.5>
        EndColor=<rgba 0.4 0.0144 0.0465 2>
        LerpFactor=&UniformAOrBFloat(A=-1, Evaluation Type=::NewEnumerator0, Recalculate Random Each Loop=true, Override Seed=false)
       ))
      B=&Multiply_VectorByFloat(Vector=<1 1 1>, Float=1.3)
     )
     "Float (Alpha)"=&Multiply_Float(
      A=2
      B=&Multiply_Float(
       A=&FloatFromCurve(
        FloatCurve=DataInterface() {propertyValues {Curve {defaultValue=1, keys [{tangentMode=RCTM_User, time=0, value=0}, {tangentMode=RCTM_User, time=1, value=1}]}}}
        CurveIndex=$Emitter.Age
        Scale Curve=&FN_FrontEndFloatModifier(Gameplay=1, Front End=0.5)
       )
       B=&FloatFromCurve(
        FloatCurve=DataInterface() {propertyValues {Curve {
           defaultValue=1
           keys [
            {interpMode=RCIM_Cubic, tangentMode=RCTM_User, time=0, value=0}
            {interpMode=RCIM_Cubic, tangentMode=RCTM_User, time=0.5, value=1}
            {tangentMode=RCTM_User, time=1, value=0}
           ]
          }}}
        Scale Curve=0.6
       )
      )
     )
    )
    Create Color Attribute=true
    Create SpriteSize Attribute=true
    Module Output:ENiagara_AttributeSamplingApplyOutput=::NewEnumerator1
    SpriteSize=&Vector2DFromFloat(Value=&Multiply_Float(
      A=&RandomRangeFloat(Minimum=150, Maximum=200)
      B=&FloatFromCurve(
       FloatCurve=DataInterface() {propertyValues {Curve {defaultValue=1, keys [{tangentMode=RCTM_User, time=0, value=0.5}, {tangentMode=RCTM_User, time=1, value=1.3}]}}}
       Scale Curve=&FN_FrontEndFloatModifier(Gameplay=1, Front End=0.8)
      )
     ))
   )
   /SetVariables(Particles.VisibilityIndex:NiagaraInt32=&MakeIntFromFloat(Float=&UniformAOrBFloat(A=0, Distribution Weight=0.3, Evaluation Type=::NewEnumerator0, Recalculate Random Each Loop=true, Override Seed=false))) {isSetParameters=true, module=SetVariables_#15}
  ]
  rendererClasses [&NiagaraSpriteRendererProperties]
  renderers [
   {rendererClass=&NiagaraSpriteRendererProperties}
   {rendererClass=&NiagaraSpriteRendererProperties, rendererIndex=1}
  ]
 }
 SmokeDetal {
  bEnabled=false
  emitterUpdateScript [
   &EmitterState(Life Cycle Mode=::NewEnumerator1, Loop Duration=1, UseLoopDelay=false)
   &SpawnRate(
    Spawn Probability=&FN_FrontEndFloatModifier(Gameplay=$User.ForwardDot, Front End=1)
    SpawnRate=6
    Use Spawn Probability=true
   )
  ]
  particleSpawnScript [
   &PartitionParticles()
   &InitializeParticle(
    Color=&LerpLinearColors(
     StartColor=<rgba 0.5 0.025 0.0646 1.5>
     EndColor=<rgba 0.4 0.0144 0.0465 2>
     LerpFactor=&UniformAOrBFloat(A=-1, Evaluation Type=::NewEnumerator0, Recalculate Random Each Loop=true, Override Seed=false)
    )
    Lifetime Max=1.5
    Lifetime Mode=::NewEnumerator1
    Mesh Scale=<1.2 1.2 1>
    Mesh Scale Mode=::NewEnumerator4
    Sprite Rotation Mode=::NewEnumerator2
    Sprite Size Mode=::NewEnumerator1
    Uniform Sprite Size Max=200
    Uniform Sprite Size Min=150
    UsePositionOffset=false
    Write Material Random=true
   )
   &CameraOffset(Camera Offset Amount=-30)
   &AddVelocity(Velocity=&MakeVector(
     X=&Multiply_Float(
      A=&SelectFloatFromArray(Direct Array Index=$Particles.PartitionParticles.Partition, Float Selection Array=DataInterface() {propertyValues {FloatData [170,-170]}})
      B=&FN_FrontEndFloatModifier(Gameplay=1, Front End=0.8)
     )
     Y=200
     Z=0
    ))
   &AddVelocity(Velocity=&FN_FrontEndVector3Modifier(Gameplay=<0 0 -50>, Front End=<0 0 40>)) {module=AddVelocity001}
  ]
  particleUpdateScript [
   &ParticleState()
   &SpriteRotationRate(Rotation Rate=&RandomRangeFloat(Minimum=-50, Maximum=50))
   &ScaleSpriteSize(
    Scale Sprite Size Mode=::NewEnumerator0
    Uniform Scale Factor=&Multiply_Float(
     A=&FloatFromCurve(FloatCurve=DataInterface() {propertyValues {Curve {defaultValue=1, keys [{tangentMode=RCTM_User, time=0, value=0.5}, {tangentMode=RCTM_User, time=1, value=1.3}]}}})
     B=1
    )
   )
   &ScaleColor(
    Scale Alpha=&Multiply_Float(
     A=&FloatFromCurve(
      FloatCurve=DataInterface() {propertyValues {Curve {defaultValue=1, keys [{tangentMode=RCTM_User, time=0, value=0}, {tangentMode=RCTM_User, time=1, value=1}]}}}
      CurveIndex=$Emitter.Age
      Scale Curve=&FN_FrontEndFloatModifier(Gameplay=1, Front End=0.5)
     )
     B=&FloatFromCurve(
      FloatCurve=DataInterface() {propertyValues {Curve {
         defaultValue=1
         keys [
          {interpMode=RCIM_Cubic, tangentMode=RCTM_User, time=0, value=0}
          {interpMode=RCIM_Cubic, tangentMode=RCTM_User, time=0.5, value=1}
          {tangentMode=RCTM_User, time=1, value=0}
         ]
        }}}
      Scale Curve=0.6
     )
    )
    Scale RGB=&Multiply_VectorByFloat(
     Vector=<1 1 1>
     Float=&FloatFromCurve(FloatCurve=DataInterface() {propertyValues {Curve {keys [<at 0 1>,<at 1 1>]}}}, Scale Curve=2)
    )
    ScaleA=true
    ScaleRGB=true
   )
   &SolveForcesAndVelocity(Clamp Velocity=false, Limit Acceleration=false)
   &SubUVAnimation(
    End Frame Range Override=30
    "Sprite Renderer "=DataInterface() {propertyValues {SpriteRenderer=&self:SmokeDetal_0.NiagaraSpriteRendererProperties_0}}
    Start Frame Range Override=&RandomRangeInt(Minimum=0, Maximum=5)
    UseEndFrame=true
    UseStartFrame=true
   )
  ]
  rendererClasses [&NiagaraSpriteRendererProperties]
  renderers [{rendererClass=&NiagaraSpriteRendererProperties}]
 }
 SppedLine {
  simTarget=GPUComputeSim
  emitterUpdateScript [
   &EmitterState(
    Delay First Loop Only=true
    Inactive Response=::NewEnumerator1
    Life Cycle Mode=::NewEnumerator1
    Loop Delay=1
    Loop Duration=1
    UseLoopDelay=true
   )
   &SpawnRate(
    Spawn Probability=&FN_FrontEndFloatModifier(Gameplay=$User.ForwardDot, Front End=1)
    SpawnRate=200
    Use Spawn Probability=true
   ) {module=SpawnRate001}
  ]
  particleSpawnScript [
   &InitializeParticle(
    Color=&MakeLinearColorFromVectorAndFloat(
     "Vector (RGB)"=<1 1 1>
     "Float (Alpha)"=&Multiply_Float(
      A=&UniformRangedFloat(Minimum=0.1, Maximum=0.2)
      B=&MakeCustomFloatFromBool(Bool=$User.bIsFrontEndPreview, True Float=0.5, False Float=1)
     )
    )
    Lifetime Max=0.1
    Lifetime Min=0.1
    Lifetime Mode=::NewEnumerator1
    Sprite Size Max=<3 10>
    Sprite Size Min=<2 2>
    Sprite Size Mode=::NewEnumerator2
    UsePositionOffset=false
    Write Material Random=true
   )
   &ShapeLocation(
    Handle Radius=0
    Large Radius=&FN_FrontEndFloatModifier(Gameplay=&RandomRangeFloat(Minimum=250, Maximum=300), Front End=220)
    MaskSpawnGroup=false
    Offset=&FN_FrontEndVector3Modifier(Gameplay=<0 0 0>, Front End=<0 -70 0>)
    Offset Mode=::NewEnumerator0
    Override Seed=false
    Rotation Axis=<0 1 0>
    Rotation Mode=::NewEnumerator1
    Shape Primitive=::NewEnumerator3
    UseOwnerScale=true
   )
   &AddVelocity(
    Module.Use Falloff Distance=false
    Origin Offset=&FN_FrontEndVector3Modifier(Gameplay=<0 0 0>, Front End=<0 -400 0>)
    Override Seed=false
    Velocity Mode=::NewEnumerator1
    Velocity Speed=&FN_FrontEndFloatModifier(Gameplay=100, Front End=20)
   ) {module=AddVelocity001}
   &AddVelocity(Velocity=&Multiply_VectorByFloat(
     Vector=&Multiply_Vector(
      A=&RandomRangeVector(Minimum=<0 50 0>, Maximum=<0 200 0>)
      B=&MakeVector(X=1, Y=&MakeCustomFloatFromBool(Bool=$User.bIsFrontEndPreview, True Float=1, False Float=1), Z=1)
     )
     Float=&FN_FrontEndFloatModifier(Gameplay=1, Front End=0.1)
    ))
  ]
  particleUpdateScript [
   &ParticleState()
   &ScaleSpriteSize(Scale Factor=&Vector2DFromCurve(
     Vector2Curve=DataInterface() {propertyValues {XCurve {keys [<at 0 0.5 RCIM_Cubic>,<at 0.5 1 RCIM_Cubic>]}, YCurve {keys [<at 0 0 RCIM_Cubic>,<at 0.5 1 RCIM_Cubic>]}}}
     ScaleCurve=&Vector2DFromFloat(Value=1)
    ))
   &SolveForcesAndVelocity(Clamp Velocity=false, Limit Acceleration=false)
   &ScaleSpriteSizeBySpeed(Max Scale Factor=<1 30>, Min Scale Factor=<0.01 0.05>, Velocity Threshold=1)
   &ScaleColor(
    Scale Alpha=&Multiply_Float(
     A=&FloatFromCurve(
      FloatCurve=DataInterface() {propertyValues {Curve {keys [
          <at 0 0 RCIM_Cubic>
          <at 0.37 1.0207>
          {arriveTangent=-2.221, interpMode=RCIM_Cubic, leaveTangent=-2.221, tangentMode=RCTM_User, time=1, value=0}
         ]}}}
      Scale Curve=&MakeCustomFloatFromBool(
       Bool=$User.bIsFrontEndPreview
       True Float=&MakeCustomFloatFromBool(Bool=$User.bIsFullyDeployed, True Float=1, False Float=0)
       False Float=&Lerp_Float(
        A=&FN_FrontEndFloatModifier(Gameplay=0, Front End=1)
        B=1
        Alpha=$System.InterpolateOverTime.InterpolateOverTime_Float.Moving Average
       )
      )
     )
     B=2
    )
    Scale RGB=&VectorFromFloat(Value=&FN_FrontEndFloatModifier(Gameplay=1, Front End=0.4))
    ScaleA=true
    ScaleRGB=true
   )
  ]
  rendererClasses [&NiagaraSpriteRendererProperties]
  renderers [{rendererClass=&NiagaraSpriteRendererProperties}]
 }
 sparks {
  emitterUpdateScript [
   &EmitterState(
    Delay First Loop Only=true
    Inactive Response=::NewEnumerator1
    Life Cycle Mode=::NewEnumerator1
    Loop Delay=1
    Loop Duration=4.6
    UseLoopDelay=true
   )
   &SpawnRate(
    Spawn Probability=&FN_FrontEndFloatModifier(Gameplay=$User.ForwardDot, Front End=1)
    SpawnRate=40
    Use Spawn Probability=true
   )
  ]
  particleSpawnScript [
   &InitializeParticle(
    Color Mode=::NewEnumerator0
    Lifetime Mode=::NewEnumerator1
    Ribbon Width=40
    Ribbon Width Mode=::NewEnumerator1
    Sprite Size Max=<1.5 7>
    Sprite Size Min=<1 3>
    Sprite Size Mode=::NewEnumerator2
    UsePositionOffset=false
    Write Material Random=true
   )
   &ShapeLocation(MaskSpawnGroup=false, Offset=<4 10 0>, Offset Mode=::NewEnumerator0, Override Seed=false, Sphere Radius=20, UseOwnerScale=true)
   /SetVariables(Particles.SizeAll=2) {isSetParameters=true}
   &AddVelocity(Velocity=&MakeVector(X=&RandomRangeFloat(Minimum=-180, Maximum=180), Y=200, Z=0)) {module=AddVelocity001}
   &AddVelocity(Velocity=&FN_FrontEndVector3Modifier(Gameplay=<0 50 -50>, Front End=<0 50 50>)) {module=AddVelocity002}
  ]
  particleUpdateScript [
   &Color(Color=&self:ScratchDynamicInput_01(
     BlotColor:LinearColor=<rgba 50 0 0.8333 1>
     CitadelColor:LinearColor=<rgba 0 16.6667 50 1>
     FrothColor:LinearColor=<rgba 0 0 0 50>
     VariantColorIndex:NiagaraInt32=&MakeIntFromFloat(Float=$User.VariantColorIndex)
    ))
   &ParticleState()
   &ScaleSpriteSize(Scale Factor=&Vector2DFromFloat(Value=&FloatFromCurve(
      FloatCurve=DataInterface() {propertyValues {Curve {keys [
          <at 0 0 RCIM_Cubic>
          {interpMode=RCIM_Cubic, leaveTangent=-3.0296, tangentMode=RCTM_Break, time=0.2, value=1.2}
          {arriveTangent=-1.8107, interpMode=RCIM_Cubic, leaveTangent=-1.8107, tangentMode=RCTM_User, time=1, value=0.2}
         ]}}}
      Scale Curve=&Multiply_Float(A=0.4, B=$Particles.SizeAll)
     ))) {module=ScaleSpriteSize001}
   &SpriteRotationRate(Rotation Rate=&RandomRangeFloat(Minimum=-70, Maximum=70))
   /SetVariables(Particles.Glow03_Color=&MakeLinearColorFromVectorAndFloat(
     "Vector (RGB)"=&MakeVectorFromLinearColorRGB(Color=&MultiplyLinearColor(
       Linear Color=&self:ScratchDynamicInput_01(
        BlotColor:LinearColor=<rgba 1 0.0112 0.0112 1>
        CitadelColor:LinearColor=<rgba 0.0112 0.3408 1 1>
        FrothColor:LinearColor=<rgba 0 0 0 2>
        VariantColorIndex:NiagaraInt32=&MakeIntFromFloat(Float=$User.VariantColorIndex)
       )
       Scale Factor=&MakeVector4_FromFloat(Float=1)
      ))
     "Float (Alpha)"=&Multiply_Float(A=0.1, B=1)
    )) {isSetParameters=true}
   /SetVariables(Particles.Glow03_Size=&Multiply_Vector2D_ByFloat(
     Float=&FloatFromCurve(
      FloatCurve=DataInterface() {propertyValues {Curve {keys [
          <at 0 0 RCIM_Cubic>
          <at 0.2 1.3 RCIM_Cubic>
          {arriveTangent=-0.1423, interpMode=RCIM_Cubic, leaveTangent=-0.1423, tangentMode=RCTM_User, time=1, value=0}
         ]}}}
      Scale Curve=&Multiply_Float(A=6, B=$Particles.SizeAll)
     )
     Vector2D=<1 2>
    )) {isSetParameters=true}
   &DynamicMaterialParameters(
    Index 0 Param 1=&MakeCustomFloatFromBool(Bool=&RandomBool(), True Float=0, False Float=0.75)
    Module.Index0Float0Write=true
    Module.Index0Float1Write=true
    Module.Index0Float2Write=true
    Module.Index0Float3Write=true
   )
   &CurlNoiseForce(Bias Noise=false, Noise Strength=50, Pan Noise=true, Pan Noise Field=<0 1 0>)
   &SolveForcesAndVelocity(Clamp Velocity=false, Limit Acceleration=false)
   &CameraOffset(Camera Offset Amount=-10) {module=CameraOffset001}
  ]
  rendererClasses [&NiagaraSpriteRendererProperties]
  renderers [
   {rendererClass=&NiagaraSpriteRendererProperties}
   {rendererClass=&NiagaraSpriteRendererProperties, rendererIndex=1}
  ]
 }
}
summary {
 systemName=NS_Glider_SheerWill
 emitters [
  {emitterName=IdleTrailOutSide, rendererClasses [&NiagaraMeshRendererProperties]}
  {bEnabled=false, emitterName=IdleSmoke, rendererClasses [&NiagaraSpriteRendererProperties]}
  {emitterName=Glow, rendererClasses [&NiagaraSpriteRendererProperties, &NiagaraLightRendererProperties]}
  {emitterName=IdleTrailInside, rendererClasses [&NiagaraMeshRendererProperties]}
  {emitterName=ForWardTrailLong, rendererClasses [&NiagaraMeshRendererProperties]}
  {emitterName=ForWardTrailShort, rendererClasses [&NiagaraMeshRendererProperties]}
  {emitterName=SmokeBlur, rendererClasses [&NiagaraSpriteRendererProperties]}
  {emitterName=sparks, rendererClasses [&NiagaraSpriteRendererProperties]}
  {bEnabled=false, emitterName=SmokeDetal, rendererClasses [&NiagaraSpriteRendererProperties]}
  {emitterName=Lightning, rendererClasses [&NiagaraSpriteRendererProperties]}
  {emitterName=SppedLine, simTarget=GPUComputeSim, rendererClasses [&NiagaraSpriteRendererProperties]}
  {emitterName=ForWardTrailLong003, rendererClasses [&NiagaraMeshRendererProperties]}
  {bEnabled=false, emitterName=IdleTrailOutSide001, rendererClasses [&NiagaraMeshRendererProperties]}
  {bEnabled=false, emitterName=SmokeBlur001, rendererClasses [&NiagaraSpriteRendererProperties]}
  {bEnabled=false, emitterName=IdleTrailOutSide002, rendererClasses [&NiagaraMeshRendererProperties]}
 ]
 userVariables [
  User.bIsFrontEnd {type=NiagaraBool}
  User.bIsFrontEndPreview {type=NiagaraBool}
  User.bIsFullyDeployed {type=NiagaraBool}
  User.ForwardDot {type=NiagaraFloat, defaultValue=0, description=""}
  User.VariantColorIndex {type=NiagaraFloat, defaultValue=0, description=""}
 ]
}
