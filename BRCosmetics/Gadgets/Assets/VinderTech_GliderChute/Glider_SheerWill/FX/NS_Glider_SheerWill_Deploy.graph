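# Niagara deploy effect for the Sheer Will glider, with colors picked per Blot, Citadel or Froth variant through User.VariantColorIndex.
# Glowing ribbon trails and GPU curl-noise sparks spawn across ten partitioned skeletal mesh bones or sockets, fading by emitter age with four partitions appearing later.
# A delayed sub-UV smoke burst adds detail.
# Front-end modifiers adjust trail width and opacity; the fixed-red duplicate trail emitter is disabled.
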NiagaraSystem /BRCosmetics/Gadgets/Assets/VinderTech_GliderChute/Glider_SheerWill/FX/NS_Glider_SheerWill_Deploy
dependencies {
 usedDataInterfaces [
  &NiagaraDataInterfaceCurve
  &NiagaraDataInterfaceArrayFloat
  &NiagaraDataInterfaceColorCurve
  &NiagaraDataInterfaceVector2DCurve
  &NiagaraDataInterfaceSpriteRendererInfo
  &NiagaraDataInterfaceSkeletalMesh
 ]
 usedDynamicInputs [
  &Vector2DFromCurve
  &RandomRangeVector2D
  &self:ScratchDynamicInput_02
  &self:ScratchDynamicInput_03
 ]
 usedModules [
  &self:SystemSpawnScript.SystemScriptSource.SystemScriptGraph.NiagaraNodeAssignment_1.SetVariables_#1
  &SystemState
  &self:SystemSpawnScript.SystemScriptSource.SystemScriptGraph.NiagaraNodeAssignment_0.SetVariables_#2
  &self:SystemSpawnScript.SystemScriptSource.SystemScriptGraph.NiagaraNodeAssignment_2.SetVariables_#3
  &self:SystemSpawnScript.SystemScriptSource.SystemScriptGraph.NiagaraNodeAssignment_3.SetVariables_#4
 ]
}
emitters {
 AbsorbTrail {
  emitterUpdateScript [
   /SetVariables(Emitter.Des:NiagaraFloat=&self:ScratchDynamicInput_09(
     Blot:NiagaraFloat=0.5
     Citadel:NiagaraFloat=0.5
     Froth:NiagaraFloat=1
     VariantColorIndex:NiagaraInt32=&MakeIntFromFloat(Float=$User.VariantColorIndex)
    )) {isSetParameters=true, module=SetVariables_#5}
   /SetVariables(Emitter.Des02:NiagaraFloat=&self:ScratchDynamicInput_10(
     Blot:NiagaraFloat=0
     Citadel:NiagaraFloat=0
     Froth:NiagaraFloat=1
     VariantColorIndex:NiagaraInt32=&MakeIntFromFloat(Float=$User.VariantColorIndex)
    )) {isSetParameters=true, module=SetVariables_#6}
   &EmitterState()
   &SpawnRate(SpawnRate=400, Use Spawn Probability=false)
  ]
  particleSpawnScript [
   &PartitionParticles(Number Of Partitions=10)
   &InitializeParticle(
    Color Mode=::NewEnumerator0
    Lifetime=0.5
    Ribbon Width=&FN_FrontEndFloatModifier(Gameplay=25, Front End=17)
    Ribbon Width Mode=::NewEnumerator1
    UsePositionOffset=false
    Write Material Random=true
   )
   &SkeletalMeshLocation(
    Bone / Socket Index=$Particles.PartitionParticles.Partition
    Bone Sampling Mode=::NewEnumerator4
    MaskSpawnGroup=false
    Skeletal Mesh=$System.Skeletal Mesh Reader01
   )
  ]
  particleUpdateScript [
   /SetVariables(Particles.BrightnessAll:NiagaraFloat=&FloatFromCurve(FloatCurve=DataInterface() {propertyValues {Curve {keys [<at 0 20>,<at 0.3241 2.1176>,<at 1 1>]}}})) {isSetParameters=true, module=SetVariables_#7}
   /SetVariables(Particles.AlphaAll:NiagaraFloat=&Multiply_Float(
     A=&Lerp_Float(
      A=1
      B=&FloatFromCurve(FloatCurve=DataInterface() {propertyValues {Curve {keys [<at 1.8 0>,<at 2.1 1>]}}}, CurveIndex=$Emitter.Age)
      Alpha=&SelectFloatFromArray(
       Direct Array Index=$Particles.PartitionParticles.Partition
       Float Selection Array=DataInterface() {propertyValues {FloatData [0,0,0,0,0,0,1,1,1,1]}}
      )
     )
     B=&FloatFromCurve(FloatCurve=DataInterface() {propertyValues {Curve {keys [<at 0.8 0>,<at 2.3 1>,<at 3 0>]}}}, CurveIndex=$Emitter.Age)
    )) {isSetParameters=true, module=SetVariables_#8}
   &ParticleState()
   &InheritSourceMovement(ApplyForce=false, ApplyPosition=true, ApplyPositionFalloff=false)
   &SolveForcesAndVelocity(Clamp Velocity=false, Limit Acceleration=false)
   &ScaleColor(
    Scale Alpha=&Multiply_Float(
     A=&FloatFromCurve(FloatCurve=DataInterface() {propertyValues {Curve {keys [<at 1 0>,<at 1.5 1>]}}}, CurveIndex=$Emitter.Age)
     B=$Particles.AlphaAll
    )
    Scale RGB=&Multiply_VectorByFloat(Vector=<1 1 1>, Float=$Particles.BrightnessAll)
    ScaleA=true
    ScaleRGB=true
   )
   &FN_Mesh_Attributes(
    Color=&MakeLinearColorFromVectorAndFloat(
     "Vector (RGB)"=&Multiply_Vector(
      A=&MakeVectorFromLinearColorRGB(Color=&self:ScratchDynamicInput_01(
        BlotColor:LinearColor=<rgba 1 0.0083 0 0.4>
        CitadelColor:LinearColor=<rgba 0 0.3333 1 0.4>
        FrothColor:LinearColor=<rgba 0 0 0 1>
        VariantColorIndex:NiagaraInt32=&MakeIntFromFloat(Float=$User.VariantColorIndex)
       ))
      B=&Multiply_VectorByFloat(Vector=<1 1 1>, Float=3)
     )
     "Float (Alpha)"=&Multiply_Float(
      A=&Multiply_Float(
       A=&FN_FrontEndFloatModifier(
        Gameplay=0.3
        Front End=&self:ScratchDynamicInput_07(
         Blot:NiagaraFloat=0.2
         Citadel:NiagaraFloat=0.4
         Froth:NiagaraFloat=0.2
         VariantColorIndex:NiagaraInt32=&MakeIntFromFloat(Float=$User.VariantColorIndex)
        )
       )
       B=&FloatFromCurve(FloatCurve=DataInterface() {propertyValues {Curve {keys [<at 1 0>,<at 1.5 1>]}}}, CurveIndex=$Emitter.Age, Scale Curve=$Particles.AlphaAll)
      )
      B=&self:ScratchDynamicInput_11(
       Blot:NiagaraFloat=1
       Citadel:NiagaraFloat=1
       Froth:NiagaraFloat=3
       VariantColorIndex:NiagaraInt32=&MakeIntFromFloat(Float=$User.VariantColorIndex)
      )
     )
    )
    Create Color Attribute=true
    Module Output=::NewEnumerator1
   )
   &FN_Mesh_Attributes(
    Color=&MakeLinearColorFromVectorAndFloat(
     "Vector (RGB)"=&Multiply_Vector(
      A=&MakeVectorFromLinearColorRGB(Color=&self:ScratchDynamicInput_01(
        BlotColor:LinearColor=&ColorFromCurve(ColorCurve=DataInterface() {propertyValues {
           AlphaCurve {keys [<at 0 1 RCIM_Cubic>,<at 1 0 RCIM_Cubic>]}
           BlueCurve {keys [<at 0 0.1127 RCIM_Cubic>,<at 1 0 RCIM_Cubic>]}
           GreenCurve {keys [<at 0 0.0854 RCIM_Cubic>,<at 1 0 RCIM_Cubic>]}
           RedCurve {keys [<at 0 2 RCIM_Cubic>,<at 1 0.2 RCIM_Cubic>]}
          }})
        CitadelColor:LinearColor=&ColorFromCurve(ColorCurve=DataInterface() {propertyValues {
           AlphaCurve {keys [<at 0 1 RCIM_Cubic>,<at 1 0 RCIM_Cubic>]}
           BlueCurve {keys [<at 0 1.3 RCIM_Cubic>,<at 1 0.2 RCIM_Cubic>]}
           GreenCurve {keys [<at 0 0.4703 RCIM_Cubic>,<at 1 0.0333 RCIM_Cubic>]}
           RedCurve {keys [<at 0 0.0555 RCIM_Cubic>,<at 1 0 RCIM_Cubic>]}
          }})
        FrothColor:LinearColor=&ColorFromCurve(ColorCurve=DataInterface() {propertyValues {
           AlphaCurve {keys [<at 0 1 RCIM_Cubic>,<at 1 0 RCIM_Cubic>]}
           BlueCurve {keys [<at 0 0.3 RCIM_Cubic>,<at 0.6717 0 RCIM_Cubic>]}
           GreenCurve {keys [<at 0 0.3 RCIM_Cubic>,<at 0.6717 0 RCIM_Cubic>]}
           RedCurve {keys [<at 0 0.3 RCIM_Cubic>,<at 0.6717 0 RCIM_Cubic>]}
          }})
        VariantColorIndex:NiagaraInt32=&MakeIntFromFloat(Float=$User.VariantColorIndex)
       ))
      B=&Multiply_VectorByFloat(
       Vector=&Multiply_VectorByFloat(Vector=<1 1 1>, Float=$Particles.BrightnessAll)
       Float=&FloatFromCurve(FloatCurve=DataInterface() {propertyValues {Curve {keys [<at 0 1>,<at 0.6432 1>,<at 1 1>]}}}, Scale Curve=3)
      )
     )
     "Float (Alpha)"=&Multiply_Float(
      A=10
      B=&FloatFromCurve(
       FloatCurve=DataInterface() {propertyValues {Curve {keys [<at 1 0>,<at 1.5 1>,<at 5 1>]}}}
       CurveIndex=$Emitter.Age
       Scale Curve=&FloatFromCurve(
        FloatCurve=DataInterface() {propertyValues {Curve {keys [<at 1 1.5>,<at 1.7 1.5>,<at 2 0>]}}}
        CurveIndex=$Emitter.Age
        Scale Curve=$Particles.AlphaAll
       )
      )
     )
    )
    Create Color Attribute=true
    Module Output=::NewEnumerator1
   ) {module=FN_Mesh_Attributes001}
   &DynamicMaterialParameters(
    Index 1 Param 1=1
    Index 1 Param 3=1
    Index 1 Param 4=1
    Module.Index0Float0Write=true
    Module.Index0Float1Write=true
    Module.Index0Float2Write=true
    Module.Index0Float3Write=true
    Module.Index1Float0Write=true
    Module.Index1Float1Write=true
    Module.Index1Float2Write=true
    Module.Index1Float3Write=true
    Write Parameter Index 1=::NewEnumerator1
   )
   /SetVariables(Particles.WidthGlow:NiagaraFloat=&Multiply_Float(A=$Particles.RibbonWidth, B=1.2)) {isSetParameters=true, module=SetVariables_#9}
   /SetVariables(Particles.WidthGlow03:NiagaraFloat=&Multiply_Float(A=$Particles.RibbonWidth, B=1.5)) {isSetParameters=true, module=SetVariables_#10}
   &ScaleRibbonWidth(Ribbon Width Scale=&FloatFromCurve(FloatCurve=DataInterface() {propertyValues {Curve {keys [<at 2.5 1>,<at 3 0.5>]}}}, CurveIndex=$Emitter.Age))
  ]
  rendererClasses [&NiagaraRibbonRendererProperties]
  renderers [
   {rendererClass=&NiagaraRibbonRendererProperties}
   {rendererClass=&NiagaraRibbonRendererProperties, rendererIndex=1}
   {rendererClass=&NiagaraRibbonRendererProperties, rendererIndex=2}
  ]
 }
 AbsorbTrail001 {
  bEnabled=false
  emitterUpdateScript [
   /SetVariables(Emitter.Des:NiagaraFloat=&self:ScratchDynamicInput_09(
     Blot:NiagaraFloat=0.5
     Citadel:NiagaraFloat=0.5
     Froth:NiagaraFloat=1
     VariantColorIndex:NiagaraInt32=&MakeIntFromFloat(Float=$User.VariantColorIndex)
    )) {isSetParameters=true, module=SetVariables_#5}
   /SetVariables(Emitter.Des02:NiagaraFloat=&self:ScratchDynamicInput_10(
     Blot:NiagaraFloat=0
     Citadel:NiagaraFloat=0
     Froth:NiagaraFloat=0
     VariantColorIndex:NiagaraInt32=&MakeIntFromFloat(Float=$User.VariantColorIndex)
    )) {isSetParameters=true, module=SetVariables_#6}
   &EmitterState()
   &SpawnRate(SpawnRate=400, Use Spawn Probability=false)
  ]
  particleSpawnScript [
   &PartitionParticles(Number Of Partitions=10)
   &InitializeParticle(
    Color Mode=::NewEnumerator0
    Lifetime=0.5
    Ribbon Width=&FN_FrontEndFloatModifier(Gameplay=25, Front End=17)
    Ribbon Width Mode=::NewEnumerator1
    UsePositionOffset=false
    Write Material Random=true
   )
   &SkeletalMeshLocation(
    Bone / Socket Index=$Particles.PartitionParticles.Partition
    Bone Sampling Mode=::NewEnumerator4
    MaskSpawnGroup=false
    Skeletal Mesh=$System.Skeletal Mesh Reader01
   )
  ]
  particleUpdateScript [
   /SetVariables(Particles.BrightnessAll:NiagaraFloat=&FloatFromCurve(FloatCurve=DataInterface() {propertyValues {Curve {keys [<at 0 20>,<at 0.3241 2.1176>,<at 1 1>]}}})) {isSetParameters=true, module=SetVariables_#7}
   /SetVariables(Particles.AlphaAll:NiagaraFloat=&Multiply_Float(
     A=&Lerp_Float(
      A=1
      B=&FloatFromCurve(FloatCurve=DataInterface() {propertyValues {Curve {keys [<at 1.8 0>,<at 2.1 1>]}}}, CurveIndex=$Emitter.Age)
      Alpha=&SelectFloatFromArray(
       Direct Array Index=$Particles.PartitionParticles.Partition
       Float Selection Array=DataInterface() {propertyValues {FloatData [0,0,0,0,0,0,1,1,1,1]}}
      )
     )
     B=&FloatFromCurve(FloatCurve=DataInterface() {propertyValues {Curve {keys [<at 0.8 0>,<at 2.3 1>,<at 3 0>]}}}, CurveIndex=$Emitter.Age)
    )) {isSetParameters=true, module=SetVariables_#8}
   &ParticleState()
   &InheritSourceMovement(ApplyForce=false, ApplyPosition=true, ApplyPositionFalloff=false)
   &SolveForcesAndVelocity(Clamp Velocity=false, Limit Acceleration=false)
   &ScaleColor(
    Scale Alpha=&Multiply_Float(
     A=&FloatFromCurve(FloatCurve=DataInterface() {propertyValues {Curve {keys [<at 1 0>,<at 1.5 1>]}}}, CurveIndex=$Emitter.Age)
     B=$Particles.AlphaAll
    )
    Scale RGB=&Multiply_VectorByFloat(Vector=<1 1 1>, Float=$Particles.BrightnessAll)
    ScaleA=true
    ScaleRGB=true
   )
   &FN_Mesh_Attributes(
    Color=&MakeLinearColorFromVectorAndFloat(
     "Vector (RGB)"=&Multiply_Vector(
      A=&MakeVectorFromLinearColorRGB(Color=<rgba 1 0.0083 0 0.4>)
      B=&Multiply_VectorByFloat(Vector=<1 1 1>, Float=3)
     )
     "Float (Alpha)"=&Multiply_Float(
      A=0.3
      B=&FloatFromCurve(FloatCurve=DataInterface() {propertyValues {Curve {keys [<at 1 0>,<at 1.5 1>]}}}, CurveIndex=$Emitter.Age, Scale Curve=$Particles.AlphaAll)
     )
    )
    Create Color Attribute=true
    Module Output=::NewEnumerator1
   )
   &FN_Mesh_Attributes(
    Color=&MakeLinearColorFromVectorAndFloat(
     "Vector (RGB)"=&Multiply_Vector(
      A=&MakeVectorFromLinearColorRGB(Color=&ColorFromCurve(ColorCurve=DataInterface() {propertyValues {
          AlphaCurve {keys [<at 0 1 RCIM_Cubic>,<at 1 0 RCIM_Cubic>]}
          BlueCurve {keys [<at 0 0.1127 RCIM_Cubic>,<at 0.6054 0 RCIM_Cubic>]}
          GreenCurve {keys [<at 0 0.0854 RCIM_Cubic>,<at 0.6054 0 RCIM_Cubic>]}
          RedCurve {keys [<at 0 2 RCIM_Cubic>,<at 0.6054 0.2 RCIM_Cubic>]}
         }}))
      B=&Multiply_VectorByFloat(
       Vector=&Multiply_VectorByFloat(Vector=<1 1 1>, Float=$Particles.BrightnessAll)
       Float=&FloatFromCurve(FloatCurve=DataInterface() {propertyValues {Curve {keys [<at 0 1>,<at 0.6432 1>,<at 1 1>]}}}, Scale Curve=3)
      )
     )
     "Float (Alpha)"=&Multiply_Float(
      A=10
      B=&FloatFromCurve(
       FloatCurve=DataInterface() {propertyValues {Curve {keys [<at 1 0>,<at 1.5 1>,<at 5 1>]}}}
       CurveIndex=$Emitter.Age
       Scale Curve=&FloatFromCurve(
        FloatCurve=DataInterface() {propertyValues {Curve {keys [<at 1 1.5>,<at 1.7 1.5>,<at 2 0>]}}}
        CurveIndex=$Emitter.Age
        Scale Curve=$Particles.AlphaAll
       )
      )
     )
    )
    Create Color Attribute=true
    Module Output=::NewEnumerator1
   ) {module=FN_Mesh_Attributes001}
   &DynamicMaterialParameters(
    Index 1 Param 1=1
    Index 1 Param 3=1
    Index 1 Param 4=1
    Module.Index0Float0Write=true
    Module.Index0Float1Write=true
    Module.Index0Float2Write=true
    Module.Index0Float3Write=true
    Module.Index1Float0Write=true
    Module.Index1Float1Write=true
    Module.Index1Float2Write=true
    Module.Index1Float3Write=true
    Write Parameter Index 1=::NewEnumerator1
   )
   /SetVariables(Particles.WidthGlow:NiagaraFloat=&Multiply_Float(A=$Particles.RibbonWidth, B=1.2)) {isSetParameters=true, module=SetVariables_#9}
   /SetVariables(Particles.WidthGlow03:NiagaraFloat=&Multiply_Float(A=$Particles.RibbonWidth, B=1.5)) {isSetParameters=true, module=SetVariables_#10}
   &ScaleRibbonWidth(Ribbon Width Scale=&FloatFromCurve(FloatCurve=DataInterface() {propertyValues {Curve {keys [<at 2.5 1>,<at 3 0.5>]}}}, CurveIndex=$Emitter.Age))
  ]
  rendererClasses [&NiagaraRibbonRendererProperties]
  renderers [
   {rendererClass=&NiagaraRibbonRendererProperties}
   {rendererClass=&NiagaraRibbonRendererProperties, rendererIndex=1}
   {rendererClass=&NiagaraRibbonRendererProperties, rendererIndex=2}
  ]
 }
 BUrstSmokeDetail {
  emitterUpdateScript [
   &EmitterState(Life Cycle Mode=::NewEnumerator1, Loop Behavior=::NewEnumerator1, Loop Delay=2.5, Loop Duration=1, UseLoopDelay=true)
   &SpawnBurst_Instantaneous(Spawn Count=6, Use Spawn Probability=false, UseLoopCountLimit=false)
  ]
  particleSpawnScript [
   &InitializeParticle(
    Color Mode=::NewEnumerator0
    Lifetime Max=1.2
    Lifetime Min=0.7
    Lifetime Mode=::NewEnumerator1
    Sprite Rotation Mode=::NewEnumerator2
    Sprite Size Mode=::NewEnumerator1
    Uniform Sprite Size Max=80
    Uniform Sprite Size Min=60
    UsePositionOffset=false
    Write Material Random=true
   )
   &CameraOffset(Camera Offset Amount=-40)
   &ShapeLocation(
    MaskSpawnGroup=false
    Non Uniform Scale=<1 0 1>
    Offset=<0 0 -10>
    Offset Mode=::NewEnumerator0
    Override Seed=false
    Sphere Radius=5
    Sphere Surface Distribution=1
    UseOwnerScale=true
   )
   &AddVelocity(
    Module.Use Falloff Distance=false
    Origin Offset=&AddVectorToPosition(Position=&RandomRangeVector(Minimum=<0 0 -5>, Maximum=<0 0 -3>), Vector=<0 0 -10>)
    Override Seed=false
    Velocity Mode=::NewEnumerator1
    Velocity Speed=200
   )
  ]
  particleUpdateScript [
   &Color(Color=&self:ScratchDynamicInput_01(
     BlotColor:LinearColor=<rgba 1 0.036 0.1163 1>
     CitadelColor:LinearColor=<rgba 0.036 0.3573 1 2>
     FrothColor:LinearColor=<rgba 0 0 0 2>
     VariantColorIndex:NiagaraInt32=&MakeIntFromFloat(Float=$User.VariantColorIndex)
    ))
   &ParticleState()
   &SpriteRotationRate(Rotation Rate=&RandomRangeFloat(Minimum=-50, Maximum=50))
   &ScaleSpriteSize(
    Scale Sprite Size Mode=::NewEnumerator0
    Uniform Scale Factor=&FloatFromCurve(FloatCurve=DataInterface() {propertyValues {Curve {keys [<at 0 1>,<at 1 1.2>]}}})
   )
   &SubUVAnimation(
    End Frame Range Override=30
    "Sprite Renderer "=DataInterface() {propertyValues {SpriteRenderer=&self:BUrstSmokeDetail_0.NiagaraSpriteRendererProperties_0}}
    Start Frame Range Override=&RandomRangeInt(Minimum=0, Maximum=5)
    UseEndFrame=true
    UseStartFrame=true
   )
   &ScaleColor(
    Color Value To Scale=$Particles.Color
    Scale Alpha=&Multiply_Float(
     A=&RandomRangeFloat(Minimum=0.5, Maximum=1)
     B=&FloatFromCurve(FloatCurve=DataInterface() {propertyValues {Curve {keys [
          {interpMode=RCIM_Cubic, tangentMode=RCTM_User, time=0, value=0}
          {arriveTangent=1.7953, interpMode=RCIM_Cubic, leaveTangent=1.7953, time=0.2, value=0}
          {interpMode=RCIM_Cubic, tangentMode=RCTM_User, time=0.5055, value=0.9075}
          {tangentMode=RCTM_User, time=1, value=0}
         ]}}})
    )
    Scale RGB=&MakeVectorFromLinearColorRGB(Color=<rgba 1 1 1 1>)
    ScaleA=true
    ScaleRGB=true
   )
   &Drag(Drag=&RandomRangeFloat(Minimum=2, Maximum=4), Use Linear Drag=true, Use Rotational Drag=true)
   &SolveForcesAndVelocity(Clamp Velocity=false, Limit Acceleration=false)
  ]
  rendererClasses [&NiagaraSpriteRendererProperties]
  renderers [{rendererClass=&NiagaraSpriteRendererProperties}]
 }
 Sparks {
  simTarget=GPUComputeSim
  emitterUpdateScript [
   &EmitterState()
   &SpawnRate(SpawnRate=500, Use Spawn Probability=false)
  ]
  particleSpawnScript [
   &PartitionParticles(Number Of Partitions=10)
   &InitializeParticle(
    Color Mode=::NewEnumerator0
    Lifetime Max=0.9
    Lifetime Min=0.4
    Lifetime Mode=::NewEnumerator1
    Sprite Size Mode=::NewEnumerator1
    Uniform Sprite Size Max=2
    Uniform Sprite Size Min=0.5
    UsePositionOffset=false
    Write Material Random=true
   )
   &SkeletalMeshLocation(
    Bone / Socket Index=$Particles.PartitionParticles.Partition
    Bone Sampling Mode=::NewEnumerator4
    MaskSpawnGroup=false
    Skeletal Mesh=$System.Skeletal Mesh Reader01
   )
   &ShapeLocation(MaskSpawnGroup=false, Override Seed=false, Sphere Radius=10, UseOwnerScale=true)
  ]
  particleUpdateScript [
   &Color(Color=&self:ScratchDynamicInput_01(
     BlotColor:LinearColor=<rgba 50 0 0.8333 1>
     CitadelColor:LinearColor=<rgba 0 16.6667 50 1>
     FrothColor:LinearColor=<rgba 0.8 0.8 0.8 1>
     VariantColorIndex:NiagaraInt32=&MakeIntFromFloat(Float=$User.VariantColorIndex)
    ))
   /SetVariables(Particles.AlphaAll:NiagaraFloat=&Multiply_Float(
     A=&Lerp_Float(
      A=1
      B=&FloatFromCurve(FloatCurve=DataInterface() {propertyValues {Curve {keys [<at 1.8 0>,<at 2.1 1>]}}}, CurveIndex=$Emitter.Age)
      Alpha=&SelectFloatFromArray(
       Direct Array Index=$Particles.PartitionParticles.Partition
       Float Selection Array=DataInterface() {propertyValues {FloatData [0,0,0,0,0,0,1,1,1,1]}}
      )
     )
     B=&FloatFromCurve(FloatCurve=DataInterface() {propertyValues {Curve {keys [<at 1 0>,<at 1.5 1>,<at 2 1>,<at 2.3 0>]}}}, CurveIndex=$Emitter.Age)
    )) {isSetParameters=true, module=SetVariables_#8}
   /SetVariables(Particles.SizeAll:Vector2f=&Multiply_Vector2D_ByFloat(
     Float=&FloatFromCurve(FloatCurve=DataInterface() {propertyValues {Curve {keys [<at 0 1>,<at 0.7 1>,<at 1 0>]}}})
     Vector2D=<1 1>
    )) {isSetParameters=true}
   &ParticleState()
   &ScaleSpriteSize(Scale Factor=$Particles.SizeAll, Scale Sprite Size Mode=::NewEnumerator2)
   &InheritSourceMovement(ApplyForce=false, ApplyPosition=true, ApplyPositionFalloff=false)
   &CurlNoiseForce(Bias Noise=false, Noise Strength=500, Pan Noise=false)
   &Drag(Drag=2, Use Linear Drag=true, Use Rotational Drag=true)
   &SolveForcesAndVelocity(Clamp Velocity=false, Limit Acceleration=false)
   &ScaleColor(
    Color Value To Scale=$Particles.Color
    Scale Alpha=&Multiply_Float(A=$Particles.AlphaAll, B=1)
    Scale RGB=&Multiply_VectorByFloat(Vector=<1 1 1>, Float=1)
    ScaleA=true
    ScaleRGB=true
   )
   &DynamicMaterialParameters(
    Index 0 Param 1=2
    Module.Index0Float0Write=true
    Module.Index0Float1Write=true
    Module.Index0Float2Write=true
    Module.Index0Float3Write=true
   )
   /SetVariables(Particles.Glow03_Color=&MakeLinearColorFromVectorAndFloat(
     "Vector (RGB)"=&MakeVectorFromLinearColorRGB(Color=&MultiplyLinearColor(
       Linear Color=&self:ScratchDynamicInput_01(
        BlotColor:LinearColor=<rgba 1 0.0112 0.0112 1>
        CitadelColor:LinearColor=<rgba 0.0112 0.3408 1 1>
        FrothColor:LinearColor=<rgba 0.3 0.3 0.3 1>
        VariantColorIndex:NiagaraInt32=&MakeIntFromFloat(Float=$User.VariantColorIndex)
       )
       Scale Factor=&MakeVector4_FromFloat(Float=1)
      ))
     "Float (Alpha)"=&Multiply_Float(A=0.1, B=$Particles.AlphaAll)
    )) {isSetParameters=true}
   /SetVariables(Particles.Glow03_Size=&Multiply_Vector2D_ByFloat(
     Float=&FloatFromCurve(
      FloatCurve=DataInterface() {propertyValues {Curve {keys [
          <at 0 0 RCIM_Cubic>
          <at 0.2 1.3 RCIM_Cubic>
          {arriveTangent=-0.1423, interpMode=RCIM_Cubic, leaveTangent=-0.1423, tangentMode=RCTM_User, time=1, value=0}
         ]}}}
      Scale Curve=&Multiply_Float(A=6, B=5)
     )
     Vector2D=&Multiply_Vector2D(A=$Particles.SizeAll, B=<1 1>)
    )) {isSetParameters=true}
  ]
  rendererClasses [&NiagaraSpriteRendererProperties]
  renderers [
   {rendererClass=&NiagaraSpriteRendererProperties}
   {rendererClass=&NiagaraSpriteRendererProperties, rendererIndex=1}
  ]
 }
}
summary {
 systemName=NS_Glider_SheerWill_Deploy
 emitters [
  {emitterName=AbsorbTrail, rendererClasses [&NiagaraRibbonRendererProperties]}
  {emitterName=Sparks, simTarget=GPUComputeSim, rendererClasses [&NiagaraSpriteRendererProperties]}
  {emitterName=BUrstSmokeDetail, rendererClasses [&NiagaraSpriteRendererProperties]}
  {bEnabled=false, emitterName=AbsorbTrail001, rendererClasses [&NiagaraRibbonRendererProperties]}
 ]
 userVariables [
  User.bIsFrontEnd {type=NiagaraBool}
  User.bIsFrontEndPreview {type=NiagaraBool}
  User.ForwardDot {type=NiagaraFloat, defaultValue=0, description=""}
  User.VariantColorIndex {type=NiagaraFloat, defaultValue=0, description=""}
 ]
}
